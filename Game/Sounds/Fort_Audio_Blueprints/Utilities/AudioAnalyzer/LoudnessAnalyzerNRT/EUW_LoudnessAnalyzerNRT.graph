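# Editor utility "Loudness Meter" that batch-measures average, min and max loudness of selected SoundWaves using AudioSynesthesia LoudnessNRT.
# InitAnalyzers builds one analyzer and MVVM viewmodel per sound from shared settings.
# Analyze Assets samples each at AnalysisSteps points across its duration and broadcasts results to its viewmodel.
# CSV export runs embedded Python, but the map it passes is filled with zeroed results rather than the computed ones.
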EditorUtilityWidgetBlueprint /Game/Sounds/Fort_Audio_Blueprints/Utilities/AudioAnalyzer/LoudnessAnalyzerNRT/EUW_LoudnessAnalyzerNRT
graphs [
 AnalyzeLoudnessNRTResultsOnViewmodels {nodes [
   AnalyzeLoudnessNRTResultsOnViewmodels() {is=title K2Node_FunctionEntry}
   > n1 = For Each Loop(Array:Array of MVVM Loudness Analyzer NRT Object References=Get ScrollboxViewmodels() -> (ScrollboxViewmodels:Array of MVVM Loudness Analyzer NRT Object References)) -> (Array Element:MVVM Loudness Analyzer NRT Object Reference)
   > Set CurrentViewmodel(CurrentViewmodel:MVVM Loudness Analyzer NRT Object Reference=@n1) -> (Output_Get:MVVM Loudness Analyzer NRT Object Reference)
   > Set CurrentTime(CurrentTime:float=0) -> (Output_Get:float)
   > Clear(TargetArray:"Array of Float (double-precision)s"=Get LoudnessValues() -> (LoudnessValues:"Array of Float (double-precision)s"))
   > n2 = For Loop(LastIndex=Get AnalysisSteps() -> (AnalysisSteps:int=0))
   n3 = "To Float (Integer)"(InInt=Get AnalysisSteps() -> (AnalysisSteps=0))
   n4 = "To Float (Integer)"(InInt=@n2)
   n5 = Get Sound!(self:Audio Analyzer NRT Object Reference=Get LoudnessNRT!(self:MVVM Loudness Analyzer NRT Object Reference=Get CurrentViewmodel() -> (CurrentViewmodel:MVVM Loudness Analyzer NRT Object Reference)) -> (LoudnessNRT:Loudness NRT Object Reference)) -> (Sound:Sound Wave Object Reference)
   > n6 = Set CurrentTime(CurrentTime=float * float(A=float / float(A=@n4, B=@n3), B=Get Duration!(self:Sound Base Object Reference=@n5) -> (Duration:float)) {is=title K2Node_PromotableOperator})
   n7 = Get CurrentViewmodel()
   > Add(
    TargetArray:"Array of Float (double-precision)s"=Get LoudnessValues()
    NewItem:"Float (double-precision) (by ref)"=GetLoudnessAtTime!(self:Loudness NRT Object Reference=Get LoudnessNRT(self=@n7), InSeconds:float=@n6) -> (OutLoudness:float)
   ) {is=title K2Node_CallArrayFunction}
   n8 = GetLoudnessResults!(execute=@n2.Completed, self, LoudnessValues:"Array of Float (double-precision)s"=Get LoudnessValues()) -> (then, LoudnessNRTResults:S Loudness Analyzer NRT Results Structure)
   > Set with Broadcast LoudnessNRTResults!(
    execute
    LoudnessNRTResults:S Loudness Analyzer NRT Results Structure=@n8
    self:MVVM Loudness Analyzer NRT Object Reference=Get CurrentViewmodel()
   ) -> (then, Output_Get:S Loudness Analyzer NRT Results Structure)
  ]}
 ConstructLoudnessAnalyzerNRT {nodes [
   n1 = ConstructLoudnessAnalyzerNRT!() -> (then, Settings:Loudness NRTSettings Object Reference, SoundWave:Sound Wave Object Reference) {is=title K2Node_FunctionEntry}
   > Is Valid(InputObject=@n1)
   > Is Valid(InputObject=@n1.SoundWave)
   > n2 = Construct Object from Class!(execute, Class=/Script/AudioSynesthesia.LoudnessNRT, self, Settings:Loudness NRTSettings Object Reference=@n1) -> (then, ReturnValue:Loudness NRT Object Reference)
   > Return Node!(execute, ReturnValue:Loudness NRT Object Reference=@n2)
   Reroute Node(InputPin:Sound Wave Object Reference=@n1.SoundWave) -> (OutputPin:Sound Wave Object Reference)
  ]}
 ConstructLoudnessAnalyzerNRTSettings {nodes [
   ConstructLoudnessAnalyzerNRTSettings() {is=title K2Node_FunctionEntry}
   > n1 = Is Valid(InputObject=Get LoudnessAnalyzerNRTSettings() -> (LoudnessAnalyzerNRTSettings:Loudness NRTSettings Object Reference))
   n2 = Construct Object from Class!(execute=@n1.Is Not Valid, Class=/Script/AudioSynesthesia.LoudnessNRTSettings, self) -> (then, ReturnValue:Loudness NRTSettings Object Reference)
   > n3 = Set with Broadcast LoudnessAnalyzerNRTSettings!(execute, LoudnessAnalyzerNRTSettings:Loudness NRTSettings Object Reference=@n2) -> (then, Output_Get:Loudness NRTSettings Object Reference)
   Return Node!(execute=[@n4.then, @n3.then], Settings:Loudness NRTSettings Object Reference=@n3)
   Return Node(execute=@n1.Is Valid, Settings=Get LoudnessAnalyzerNRTSettings())
   n4 = Set with Broadcast LoudnessNRTSettings!(
    execute
    LoudnessNRTSettings:Loudness NRTSettings Object Reference=@n2
    self:MVVM Loudness Analyzer NRT Settings Object Reference=Get MVVM_LoudnessAnalyzerNRT_Settings() -> (MVVM_LoudnessAnalyzerNRT_Settings:MVVM Loudness Analyzer NRT Settings Object Reference)
   ) -> (then, Output_Get:Loudness NRTSettings Object Reference)
   Get LoudnessNRTSettings!(self:MVVM Loudness Analyzer NRT Settings Object Reference=Get MVVM_LoudnessAnalyzerNRT_Settings()) -> (LoudnessNRTSettings:Loudness NRTSettings Object Reference)
   Get LoudnessNRTSettings(self=Get MVVM_LoudnessAnalyzerNRT_Settings())
  ]}
 ConstructLoudnessAnalyzerNRTViewmodel {nodes [
   n1 = ConstructLoudnessAnalyzerNRTViewmodel!() -> (then, LoudnessNRT:Loudness NRT Object Reference) {is=title K2Node_FunctionEntry}
   > n2 = Construct Object from Class!(
    execute
    Class=&/MVVM_LoudnessAnalyzerNRT.MVVM_LoudnessAnalyzerNRT_C
    self
    LoudnessNRT:Loudness NRT Object Reference=@n1
   ) -> (then, ReturnValue:MVVM Loudness Analyzer NRT Object Reference)
   > Return Node!(execute, Viewmodel:MVVM Loudness Analyzer NRT Object Reference=@n2)
  ]}
 EventGraph {nodes [
   "On Clicked (btn_AnalyzeAssets)"() {is=title K2Node_ComponentBoundEvent}
   > HandleOnAnalyzeAssetsClicked()
   "On Property Changed (AnalyzerSettingsDetailsView)"!() -> (OutputDelegate, then, PropertyName:Name=None) {is=title K2Node_ComponentBoundEvent}
   > Broadcast Field Value Changed!(execute, Object:Object Reference=Self-Reference(), FieldId:Field Notification Id Structure=(FieldName="LoudnessAnalyzerNRTSettings")) -> (then)
   "On Clicked (btn_WriteResultsToFile)"() {is=title K2Node_ComponentBoundEvent}
   > HandleOnWriteResultsToCSVClicked()
   "On Property Changed (SoundWaveDetailsView)"!() -> (OutputDelegate, then, PropertyName:Name=None) {is=title K2Node_ComponentBoundEvent}
   > ResetAnalyzers()
   > InitAnalyzers()
   n1 = "On Text Committed (CsvPathTextBox)"!() -> (OutputDelegate, then, Text:"Text (by ref)", CommitMethod:ETextCommit Enum=Default) {is=title K2Node_ComponentBoundEvent}
   > Set with Broadcast CsvOutputFilePath!(execute, CsvOutputFilePath:String="To String (Text)"(InText=@n1.Text)) -> (then, Output_Get:String)
   n2 = "On Text Committed (CsvNameTextBox)"!() -> (OutputDelegate, then, Text:"Text (by ref)", CommitMethod:ETextCommit Enum=Default) {is=title K2Node_ComponentBoundEvent}
   > Set with Broadcast CsvOutputFileName!(execute, CsvOutputFileName:String="To String (Text)"(InText=@n2.Text)) -> (then, Output_Get:String)
   Event Construct()
   > Parent: Construct()
   > DelayUntilNextTick!(execute, WorldContextObject:Object Reference) -> (then)
   > InitAnalyzers()
   n3 = Event PreConstruct()
   > Parent: PreConstruct(IsDesignTime=@n3.IsDesignTime)
   > SetTargetObjectForDetailViews()
   > InitText()
  ]}
 GetCurveTypeAsText {nodes [
   GetCurveTypeAsText() {is=title K2Node_FunctionEntry}
   n1 = Get LoudnessAnalyzerNRTSettings()
   > Return Node!(
    execute
    ReturnValue:Text=To Text!(Value:ELoudnessNRTCurveTypeEnum Enum=Get CurveType!(self:Loudness NRTSettings Object Reference=@n1) -> (CurveType:ELoudnessNRTCurveTypeEnum Enum=A)) -> (ReturnValue:Text) {is=title K2Node_GenericToText}
   )
  ]}
 GetLoudnessResults {nodes [
   n1 = GetLoudnessResults!() -> (then, LoudnessValues:"Array of Float (double-precision)s") {is=title K2Node_FunctionEntry}
   > n2 = Set InLoudnessValues(InLoudnessValues:"Array of Float (double-precision)s"=@n1) -> (Output_Get:"Array of Float (double-precision)s")
   > Branch(Condition=Is Not Empty!(TargetArray:"Array of Float (double-precision)s"=@n2) -> (ReturnValue:bool))
   > n3 = For Each Loop(Array:"Array of Float (double-precision)s"=Get InLoudnessValues() -> (InLoudnessValues:"Array of Float (double-precision)s")) -> (Array Element:double)
   n4 = Get InLoudnessValues()
   n5 = "To Float (Integer)"(InInt=Length(TargetArray:"Array of Float (double-precision)s"=Get InLoudnessValues()))
   n6 = Get Sum() -> (Sum:double)
   Return Node!(
    execute=@n3.Completed
    LoudnessNRTResults:S Loudness Analyzer NRT Results Structure=Make S Loudness Analyzer NRT Results!(
     LoudnessAverage_3_#1:double=float / float(A=@n6, B=@n5)
     LoudnessMin_5_#2:double=MinOfFloatArray!(FloatArray:"Array of Float (single-precision)s"=@n4) -> (IndexOfMinValue:int=0, MinValue:float).MinValue
     LoudnessMax_7_#3:double=MaxOfFloatArray!(FloatArray:"Array of Float (single-precision)s"=@n4) -> (IndexOfMaxValue:int=0, MaxValue:float).MaxValue
    ) -> (S_LoudnessAnalyzerNRT_Results:S Loudness Analyzer NRT Results Structure) {is=title K2Node_MakeStruct}
   )
   > Set Sum(Sum:double=float + float(A=@n3, B=Get Sum())) -> (Output_Get:double)
  ]}
 HandleOnAnalyzeAssetsClicked {nodes [
   HandleOnAnalyzeAssetsClicked() {is=title K2Node_FunctionEntry}
   > AnalyzeLoudnessNRTResultsOnViewmodels() {is=title K2Node_CallFunction}
  ]}
 HandleOnWriteResultsToCSVClicked {nodes [
   HandleOnWriteResultsToCSVClicked() {is=title K2Node_FunctionEntry}
   > WriteResultsToCSVFile!(execute, self) -> (then, CSVFileContents:String)
  ]}
 InitAnalyzers {nodes [
   InitAnalyzers() {is=title K2Node_FunctionEntry}
   > n1 = ConstructLoudnessAnalyzerNRTSettings!(execute, self) -> (then, Settings:Loudness NRTSettings Object Reference) {is=title K2Node_CallFunction}
   > Is Valid(InputObject=@n1)
   > n2 = "For Each Loop (Set)"!(execute:Exec, SetPin:Set of Sound Wave Object References=Get SoundWaves() -> (SoundWaves:Set of Sound Wave Object References), BreakPin:Exec) -> (then:Exec, ValuePin:Sound Wave Object Reference, CompletedPin:Exec) {is=title K2Node_SetForEach}
   > Is Valid(InputObject=@n2)
   > n3 = ConstructLoudnessAnalyzerNRT!(execute, self, Settings:Loudness NRTSettings Object Reference=Get LoudnessAnalyzerNRTSettings(), SoundWave:Sound Wave Object Reference=@n2) -> (then, ReturnValue:Loudness NRT Object Reference) {is=title K2Node_CallFunction}
   > Is Valid(InputObject=@n3)
   > n4 = Set CurrentNRT(CurrentNRT:Loudness NRT Object Reference=@n3) -> (Output_Get:Loudness NRT Object Reference)
   > n5 = ConstructLoudnessAnalyzerNRTViewmodel!(execute, self, LoudnessNRT:Loudness NRT Object Reference=@n4) -> (then, Viewmodel:MVVM Loudness Analyzer NRT Object Reference) {is=title K2Node_CallFunction}
   > Is Valid(InputObject=@n5)
   > n6 = Set CurrentViewmodel(CurrentViewmodel=@n5)
   > Set with Broadcast LoudnessNRT!(
    execute
    LoudnessNRT:Loudness NRT Object Reference=Get CurrentNRT() -> (CurrentNRT:Loudness NRT Object Reference)
    self:MVVM Loudness Analyzer NRT Object Reference=@n6
   ) -> (then, Output_Get:Loudness NRT Object Reference)
   > Add(
    TargetArray:Array of MVVM Loudness Analyzer NRT Object References=Get ScrollboxViewmodels()
    NewItem:"MVVM Loudness Analyzer NRT Object Reference (by ref)"=Get CurrentViewmodel()
   )
   Broadcast Field Value Changed(execute=@n2.CompletedPin, Object=Self-Reference(), FieldId=(FieldName="ScrollboxViewmodels"))
   > n7 = For Each Loop(Array:Array of MVVM Loudness Analyzer NRT Object References=Get ScrollboxViewmodels()) -> (Array Element:MVVM Loudness Analyzer NRT Object Reference)
   > Broadcast Fields Value Changed!(
    execute
    Object:Object Reference=@n7
    FieldIds:Array of Field Notification Id Structures=Make Array!(
     "[0]":Field Notification Id Structure=Make Field Notification Id!(FieldName:Name=LoudnessNRT) -> (FieldNotificationId:Field Notification Id Structure) {is=title K2Node_MakeStruct}
     "[1]":Field Notification Id Structure=Make Field Notification Id(FieldName=GetSoundWaveNameFromNRT)
    ) -> (Array:Array of Field Notification Id Structures)
   ) -> (then)
  ]}
 InitText {nodes [
   InitText() {is=title K2Node_FunctionEntry}
   > Set Text(
    Text="To Text (String)"(InString=Get DefaultCsvOutputFilePathDefault() -> (DefaultCsvOutputFilePathDefault:String))
    self:Text Box Object Reference=Get CsvPathTextBox() -> (CsvPathTextBox:UEFN Editable Text Box Object Reference)
   )
   > Set Text(
    Text="To Text (String)"(InString=Get DefaultCsvOutputFileName() -> (DefaultCsvOutputFileName:String))
    self=Get CsvNameTextBox() -> (CsvNameTextBox:UEFN Editable Text Box Object Reference)
   )
  ]}
 ResetAnalyzers {nodes [
   ResetAnalyzers() {is=title K2Node_FunctionEntry}
   > Clear(TargetArray:Array of MVVM Loudness Analyzer NRT Object References=Get ScrollboxViewmodels())
   > Broadcast Field Value Changed(Object=Self-Reference(), FieldId=(FieldName="ScrollboxViewmodels"))
  ]}
 SetTargetObjectForDetailViews {nodes [
   SetTargetObjectForDetailViews() {is=title K2Node_FunctionEntry}
   > SetObject(self=Get SoundWaveDetailsView() -> (SoundWaveDetailsView:Details View Object Reference), NewObject=Self-Reference())
   > SetObject(self=Get AnalyzerSettingsDetailsView() -> (AnalyzerSettingsDetailsView:Details View Object Reference), NewObject=Self-Reference())
  ]}
 UpdateSettingsDisplay {nodes [
   UpdateSettingsDisplay() {is=title K2Node_FunctionEntry}
   n1 = Get LoudnessAnalyzerNRTSettings()
   > Set Text(
    Text="To Text (Float)"(
     Value=Get AnalysisPeriod!(self:Loudness NRTSettings Object Reference=@n1) -> (AnalysisPeriod:float)
     MinimumIntegralDigits=1
     MaximumIntegralDigits=324
     MinimumFractionalDigits=2
     MaximumFractionalDigits=2
    )
    self:Text Object Reference=Get Value_AnalysisPeriod() -> (Value_AnalysisPeriod:Text Object Reference)
   )
   > Set Text(
    Text="To Text (Float)"(
     Value=Get MinimumFrequency!(self:Loudness NRTSettings Object Reference=@n1) -> (MinimumFrequency:float)
     MinimumIntegralDigits=1
     MaximumIntegralDigits=324
     MinimumFractionalDigits=2
     MaximumFractionalDigits=2
    )
    self=Get Value_MinFrequency() -> (Value_MinFrequency:Text Object Reference)
   )
   > Set Text(
    Text="To Text (Float)"(
     Value=Get MaximumFrequency!(self:Loudness NRTSettings Object Reference=@n1) -> (MaximumFrequency:float)
     MinimumIntegralDigits=1
     MaximumIntegralDigits=324
     MinimumFractionalDigits=2
     MaximumFractionalDigits=2
    )
    self=Get Value_MaxFrequency() -> (Value_MaxFrequency:Text Object Reference)
   )
   > Set Text(Text=To Text(Value=Get CurveType(self=@n1) -> (CurveType=A)), self=Get Value_CurveType() -> (Value_CurveType:Text Object Reference))
   > Set Text(
    Text="To Text (Float)"(
     Value=Get NoiseFloorDb!(self:Loudness NRTSettings Object Reference=@n1) -> (NoiseFloorDb:float)
     MinimumIntegralDigits=1
     MaximumIntegralDigits=324
     MinimumFractionalDigits=2
     MaximumFractionalDigits=2
    )
    self=Get Value_NoiseFloor() -> (Value_NoiseFloor:Text Object Reference)
   )
  ]}
 WriteResultsToCSVFile {nodes [
   WriteResultsToCSVFile() {is=title K2Node_FunctionEntry}
   > n1 = For Each Loop(Array=Get ScrollboxViewmodels())
   > Add!(
    execute
    TargetMap:Map of Sound Wave Object References to S Loudness Analyzer NRT Results Structures=Get SoundWaveToLoudnessNRTResultsMap() -> (SoundWaveToLoudnessNRTResultsMap:Map of Sound Wave Object References to S Loudness Analyzer NRT Results Structures)
    Key:"Sound Wave Object Reference (by ref)"=Get Sound(self=Get LoudnessNRT(self=@n1))
    Value:"S Loudness Analyzer NRT Results Structure (by ref)"=Make S Loudness Analyzer NRT Results(
     LoudnessAverage_3_#1=0
     LoudnessMin_5_#2=0
     LoudnessMax_7_#3=0
    )
   ) -> (then) {is=title K2Node_CallFunction}
   n2 = Get CsvNameTextBox()
   n3 = Execute Python Script!(
    execute=@n1.Completed
    PythonScript="import csv\r\nimport os\r\nimport unreal\r\n\r\n## CSV Output Path\r\nProjectDir  = unreal.SystemLibrary.get_project_directory()\r\nCsvFilePath = os.path.join(ProjectDir, csv_file_output_path, csv_file_output_name + \".csv\") ##type: ignore\r\n\r\n## Write CSV File\r\nwith open(CsvFilePath, mode='w', newline='') as CsvFile: ##type: ignore\r\n    Writer = csv.writer(CsvFile)\r\n\r\n    Writer.writerow(['Soundwave', 'LoudnessAverage', 'LoudnessMin', 'LoudnessMax'])\r\n    for Soundwave, LoudnessNRTResult in soundwaves_to_loudness_results_map.items(): ## type: ignore\r\n        Writer.writerow([\r\n            Soundwave.get_name(),\r\n            LoudnessNRTResult.get_editor_property(\"LoudnessAverage\"),\r\n            LoudnessNRTResult.get_editor_property(\"LoudnessMin\"),\r\n            LoudnessNRTResult.get_editor_property(\"LoudnessMax\")\r\n        ])\r\n\r\n\r\ncsv_file_contents = \"\"\r\n\r\n## Output contents of CSV file as string\r\nwith open (CsvFilePath, \"r\") as CsvFile:\r\n    csv_file_contents = CsvFile.read()\r\n"
    soundwaves_to_loudness_results_map:Map of Sound Wave Object References to S Loudness Analyzer NRT Results Structures=Get SoundWaveToLoudnessNRTResultsMap()
    csv_file_output_path:String=Get CsvOutputFilePath() -> (CsvOutputFilePath:String)
    csv_file_output_name:String="To String (Text)"(InText=Get Text(self=@n2))
   ) -> (then, ReturnValue, csv_file_contents:String) {inputs [soundwaves_to_loudness_results_map, csv_file_output_path, csv_file_output_name], outputs [csv_file_contents]}
   > Return Node!(execute, CSVFileContents:String=@n3.csv_file_contents)
   Reroute Node(InputPin:MVVM Loudness Analyzer NRT Object Reference=@n1) -> (OutputPin:MVVM Loudness Analyzer NRT Object Reference)
  ]}
]
variables [
 AnalysisSteps {type=integer, default=2000}
 CsvOutputFileName {type=string, default=LoudnessNRTResults}
 CsvOutputFilePath {type=string, default=Content/Sounds/Fort_Audio_Blueprints/Utilities/AudioAnalyzer/LoudnessAnalyzerNRT/CSV/}
 DefaultCsvOutputFileName {type=string, default=LoudnessNRTResults}
 DefaultCsvOutputFilePathDefault {type=string, default=Content/Sounds/Fort_Audio_Blueprints/Utilities/AudioAnalyzer/LoudnessAnalyzerNRT/CSV/}
 Loudness NRTResults {type=S_LoudnessAnalyzerNRT_Results, default {loudnessAverage=0, loudnessMax=0, loudnessMin=0}}
 LoudnessAnalyzerNRTSettings {type=LoudnessNRTSettings, default=null}
 ScrollboxViewmodels {type="array<MVVM_LoudnessAnalyzerNRT_C>", default []}
 SoundWaves {type="array<SoundWave>", default []}
]
widgets {
 info {rootWidgetClass=&GridPanel, widgetCount=69}
 widgets [
  MainGridPanel {widgetClassPath=&GridPanel, widget=MainGridPanel}
  MainBorder {
   widgetClassPath=&Border
   parent=MainGridPanel
   widget=MainBorder
   properties {brushColor=<rgba 0 0 0 1>}
   slot {class=GridSlot, properties {columnSpan=5, rowSpan=19}}
  }
  Text_Header {
   widgetClassPath=&UEFN_TextBlock
   parent=MainGridPanel
   widget=Text_Header
   properties {
    text=Loudness Meter
    font {
     bForceMonospaced=false
     bMaterialIsStencil=false
     fontMaterial=null
     fontObject=&HeadingNow
     letterSpacing=56
     monospacedWidth=1
     size=27
     skewAmount=0
     typefaceFontName=86 Bold
     outlineSettings {
      bApplyOutlineToDropShadows=false
      bMiteredCorners=false
      bSeparateFillAlpha=false
      outlineMaterial=null
      outlineSize=0
      outlineColor=<rgba 0 0 0 1>
     }
    }
   }
   slot {class=GridSlot, properties {column=1, columnSpan=1, horizontalAlignment=HAlign_Center, row=1}}
  }
  TitleBar {
   widgetClassPath=&Image
   parent=MainGridPanel
   widget=TitleBar
   properties {renderTransform {angle=0, scale=<1 0.25>, shear=<0 0>, translation=<0 0>}}
   slot {class=GridSlot, properties {column=1, row=2}}
  }
  Spacer_MainBorder_TopLeft {
   widgetClassPath=&Spacer
   parent=MainGridPanel
   widget=Spacer_MainBorder_TopLeft
   properties {size=<4.7517 3.946>}
   slot {class=GridSlot, properties {padding=<ltrb 0 0 20 20>}}
  }
  Spacer_MainBorder_TopMiddle {
   widgetClassPath=&Spacer
   parent=MainGridPanel
   widget=Spacer_MainBorder_TopMiddle
   slot {class=GridSlot, properties {column=2, padding=<ltrb 0 0 20 20>}}
  }
  Spacer_MainBorder_TopRight {
   widgetClassPath=&Spacer
   parent=MainGridPanel
   widget=Spacer_MainBorder_TopRight
   properties {size=<5.4846 1>}
   slot {class=GridSlot, properties {column=4, padding=<ltrb 0 0 20 20>}}
  }
  Spacer_MainBorder_BottomLeft {
   widgetClassPath=&Spacer
   parent=MainGridPanel
   widget=Spacer_MainBorder_BottomLeft
   properties {size=<1 7.0774>}
   slot {class=GridSlot, properties {row=17, padding=<ltrb 0 0 20 20>}}
  }
  Spacer_Header_Bottom {
   widgetClassPath=&Spacer
   parent=MainGridPanel
   widget=Spacer_Header_Bottom
   properties {size=<1 -10.477>}
   slot {class=GridSlot, properties {column=1, row=3, padding=<ltrb 0 0 20 20>}}
  }
  ButtonsGridPanel {
   widgetClassPath=&GridPanel
   parent=MainGridPanel
   widget=ButtonsGridPanel
   slot {class=GridSlot, properties {column=1, horizontalAlignment=HAlign_Right, row=16, verticalAlignment=VAlign_Center, padding=<ltrb 20 0 0 0>}}
  }
  btn_AnalyzeAssets {
   widgetClassPath=&Button
   parent=ButtonsGridPanel
   widget=btn_AnalyzeAssets
   properties {widgetStyle {
     clickedSlateSound {resourceObject=null}
     disabled {
      bIsDynamicallyLoaded=false
      drawAs=RoundedBox
      imageType=NoImage
      mirroring=NoMirror
      resourceName=None
      resourceObject=null
      tiling=NoTile
      imageSize=<0 0>
      margin=<ltrb 0 0 0 0>
      outlineSettings {
       bUseBrushTransparency=false
       roundingType=FixedRadius
       width=1
       color {colorUseRule=UseColor_ColorTable, specifiedColor=<rgba 0 0 0 0>}
       cornerRadii=<4 4 4 4>
      }
      tintColor {colorUseRule=UseColor_Specified, specifiedColor=<rgba 0.0395 0.0395 0.0395 1>}
      uVRegion {bIsValid=false, max=<0 0>, min=<0 0>}
     }
     disabledForeground {colorUseRule=UseColor_ColorTable, specifiedColor=<rgba 0 0 0 0>}
     hovered {
      bIsDynamicallyLoaded=false
      drawAs=RoundedBox
      imageType=NoImage
      mirroring=NoMirror
      resourceName=None
      resourceObject=null
      tiling=NoTile
      imageSize=<32 32>
      margin=<ltrb 0 0 0 0>
      outlineSettings {
       bUseBrushTransparency=false
       roundingType=FixedRadius
       width=1
       color {colorUseRule=UseColor_ColorTable, specifiedColor=<rgba 0 0 0 0>}
       cornerRadii=<4 4 4 4>
      }
      tintColor {colorUseRule=UseColor_Specified, specifiedColor=<rgba 0.0953 0.0953 0.0953 1>}
      uVRegion {bIsValid=false, max=<0 0>, min=<0 0>}
     }
     hoveredForeground {colorUseRule=UseColor_ColorTable, specifiedColor=<rgba 0 0 0 0>}
     hoveredSlateSound {resourceObject=null}
     normal {
      bIsDynamicallyLoaded=false
      drawAs=RoundedBox
      imageType=NoImage
      mirroring=NoMirror
      resourceName=None
      resourceObject=null
      tiling=NoTile
      imageSize=<32 32>
      margin=<ltrb 0 0 0 0>
      outlineSettings {
       bUseBrushTransparency=false
       roundingType=FixedRadius
       width=1
       color {colorUseRule=UseColor_ColorTable, specifiedColor=<rgba 0 0 0 0>}
       cornerRadii=<4 4 4 4>
      }
      tintColor {colorUseRule=UseColor_Specified, specifiedColor=<rgba 0.001 0.3438 0.0321 1>}
      uVRegion {bIsValid=false, max=<0 0>, min=<0 0>}
     }
     normalForeground {colorUseRule=UseColor_ColorTable, specifiedColor=<rgba 0 0 0 0>}
     normalPadding=<ltrb 12 1.5 12 1.5>
     pressed {
      bIsDynamicallyLoaded=false
      drawAs=RoundedBox
      imageType=NoImage
      mirroring=NoMirror
      resourceName=None
      resourceObject=null
      tiling=NoTile
      imageSize=<32 32>
      margin=<ltrb 0 0 0 0>
      outlineSettings {
       bUseBrushTransparency=false
       roundingType=FixedRadius
       width=1
       color {colorUseRule=UseColor_ColorTable, specifiedColor=<rgba 0 0 0 0>}
       cornerRadii=<4 4 4 4>
      }
      tintColor {colorUseRule=UseColor_Specified, specifiedColor=<rgba 0.0284 0.0284 0.0284 1>}
      uVRegion {bIsValid=false, max=<0 0>, min=<0 0>}
     }
     pressedForeground {colorUseRule=UseColor_ColorTable, specifiedColor=<rgba 0 0 0 0>}
     pressedPadding=<ltrb 12 2.5 12 0.5>
     pressedSlateSound {resourceObject=null}
    }}
   slot {class=GridSlot, properties {column=1}}
  }
  UEFN_TextBlock_C_1 {
   widgetClassPath=&UEFN_TextBlock
   parent=btn_AnalyzeAssets
   widget=UEFN_TextBlock_C_1
   properties {
    text=Analyze Assets
    font {
     bForceMonospaced=false
     bMaterialIsStencil=false
     fontMaterial=null
     fontObject=&BurbankSmall
     letterSpacing=0
     monospacedWidth=1
     size=15
     skewAmount=0
     typefaceFontName=Black
     outlineSettings {
      bApplyOutlineToDropShadows=false
      bMiteredCorners=false
      bSeparateFillAlpha=false
      outlineMaterial=null
      outlineSize=0
      outlineColor=<rgba 0 0 0 1>
     }
    }
   }
   slot {class=ButtonSlot, properties {verticalAlignment=VAlign_Fill, padding=<ltrb 10 5 10 5>}}
  }
  ButtonsGridPanel_1 {
   widgetClassPath=&GridPanel
   parent=MainGridPanel
   widget=ButtonsGridPanel_1
   slot {class=GridSlot, properties {column=3, horizontalAlignment=HAlign_Right, row=16, verticalAlignment=VAlign_Center}}
  }
  Spacer {widgetClassPath=&Spacer, parent=ButtonsGridPanel_1, widget=Spacer, properties {size=<70.3326 1>}, slot {class=GridSlot}}
  Spacer_2 {
   widgetClassPath=&Spacer
   parent=ButtonsGridPanel_1
   widget=Spacer_2
   properties {size=<70.3326 1>}
   slot {class=GridSlot, properties {column=1, row=1}}
  }
  Spacer_1 {
   widgetClassPath=&Spacer
   parent=ButtonsGridPanel_1
   widget=Spacer_1
   properties {size=<16.3019 1>}
   slot {class=GridSlot, properties {column=3}}
  }
  btn_WriteResultsToFile {
   widgetClassPath=&Button
   parent=ButtonsGridPanel_1
   widget=btn_WriteResultsToFile
   properties {widgetStyle {
     clickedSlateSound {resourceObject=null}
     disabled {
      bIsDynamicallyLoaded=false
      drawAs=RoundedBox
      imageType=NoImage
      mirroring=NoMirror
      resourceName=None
      resourceObject=null
      tiling=NoTile
      imageSize=<0 0>
      margin=<ltrb 0 0 0 0>
      outlineSettings {
       bUseBrushTransparency=false
       roundingType=FixedRadius
       width=1
       color {colorUseRule=UseColor_ColorTable, specifiedColor=<rgba 0 0 0 0>}
       cornerRadii=<4 4 4 4>
      }
      tintColor {colorUseRule=UseColor_Specified, specifiedColor=<rgba 0.0395 0.0395 0.0395 1>}
      uVRegion {bIsValid=false, max=<0 0>, min=<0 0>}
     }
     disabledForeground {colorUseRule=UseColor_ColorTable, specifiedColor=<rgba 0 0 0 0>}
     hovered {
      bIsDynamicallyLoaded=false
      drawAs=RoundedBox
      imageType=NoImage
      mirroring=NoMirror
      resourceName=None
      resourceObject=null
      tiling=NoTile
      imageSize=<32 32>
      margin=<ltrb 0 0 0 0>
      outlineSettings {
       bUseBrushTransparency=false
       roundingType=FixedRadius
       width=1
       color {colorUseRule=UseColor_ColorTable, specifiedColor=<rgba 0 0 0 0>}
       cornerRadii=<4 4 4 4>
      }
      tintColor {colorUseRule=UseColor_Specified, specifiedColor=<rgba 0.0953 0.0953 0.0953 1>}
      uVRegion {bIsValid=false, max=<0 0>, min=<0 0>}
     }
     hoveredForeground {colorUseRule=UseColor_ColorTable, specifiedColor=<rgba 0 0 0 0>}
     hoveredSlateSound {resourceObject=null}
     normal {
      bIsDynamicallyLoaded=false
      drawAs=RoundedBox
      imageType=NoImage
      mirroring=NoMirror
      resourceName=None
      resourceObject=null
      tiling=NoTile
      imageSize=<32 32>
      margin=<ltrb 0 0 0 0>
      outlineSettings {
       bUseBrushTransparency=false
       roundingType=FixedRadius
       width=1
       color {colorUseRule=UseColor_ColorTable, specifiedColor=<rgba 0 0 0 0>}
       cornerRadii=<4 4 4 4>
      }
      tintColor {colorUseRule=UseColor_Specified, specifiedColor=<rgba 0 0.227 0.7011 1>}
      uVRegion {bIsValid=false, max=<0 0>, min=<0 0>}
     }
     normalForeground {colorUseRule=UseColor_ColorTable, specifiedColor=<rgba 0 0 0 0>}
     normalPadding=<ltrb 12 1.5 12 1.5>
     pressed {
      bIsDynamicallyLoaded=false
      drawAs=RoundedBox
      imageType=NoImage
      mirroring=NoMirror
      resourceName=None
      resourceObject=null
      tiling=NoTile
      imageSize=<32 32>
      margin=<ltrb 0 0 0 0>
      outlineSettings {
       bUseBrushTransparency=false
       roundingType=FixedRadius
       width=1
       color {colorUseRule=UseColor_ColorTable, specifiedColor=<rgba 0 0 0 0>}
       cornerRadii=<4 4 4 4>
      }
      tintColor {colorUseRule=UseColor_Specified, specifiedColor=<rgba 0.0284 0.0284 0.0284 1>}
      uVRegion {bIsValid=false, max=<0 0>, min=<0 0>}
     }
     pressedForeground {colorUseRule=UseColor_ColorTable, specifiedColor=<rgba 0 0 0 0>}
     pressedPadding=<ltrb 12 2.5 12 0.5>
     pressedSlateSound {resourceObject=null}
    }}
   slot {class=GridSlot, properties {column=4, rowSpan=2}}
  }
  UEFN_TextBlock_C {
   widgetClassPath=&UEFN_TextBlock
   parent=btn_WriteResultsToFile
   widget=UEFN_TextBlock_C
   properties {
    text=Write Results To CSV
    font {
     bForceMonospaced=false
     bMaterialIsStencil=false
     fontMaterial=null
     fontObject=&BurbankSmall
     letterSpacing=0
     monospacedWidth=1
     size=15
     skewAmount=0
     typefaceFontName=Black
     outlineSettings {
      bApplyOutlineToDropShadows=false
      bMiteredCorners=false
      bSeparateFillAlpha=false
      outlineMaterial=null
      outlineSize=0
      outlineColor=<rgba 0 0 0 1>
     }
    }
   }
   slot {class=ButtonSlot, properties {padding=<ltrb 10 5 10 5>}}
  }
  UEFN_TextBlock_C_220 {
   widgetClassPath=&UEFN_TextBlock
   parent=ButtonsGridPanel_1
   widget=UEFN_TextBlock_C_220
   properties {
    text="Folder: "
    colorAndOpacity {colorUseRule=UseColor_Specified, specifiedColor=<rgba 0.3385 0.3385 0.3385 1>}
    font {
     bForceMonospaced=false
     bMaterialIsStencil=false
     fontMaterial=null
     fontObject=&BurbankSmall
     letterSpacing=0
     monospacedWidth=1
     size=12
     skewAmount=0
     typefaceFontName=Black
     outlineSettings {
      bApplyOutlineToDropShadows=false
      bMiteredCorners=false
      bSeparateFillAlpha=false
      outlineMaterial=null
      outlineSize=0
      outlineColor=<rgba 0 0 0 1>
     }
    }
   }
   slot {class=GridSlot}
  }
  UEFN_TextBlock_C_2 {
   widgetClassPath=&UEFN_TextBlock
   parent=ButtonsGridPanel_1
   widget=UEFN_TextBlock_C_2
   properties {
    text=Name:
    colorAndOpacity {colorUseRule=UseColor_Specified, specifiedColor=<rgba 0.3385 0.3385 0.3385 1>}
    font {
     bForceMonospaced=false
     bMaterialIsStencil=false
     fontMaterial=null
     fontObject=&BurbankSmall
     letterSpacing=0
     monospacedWidth=1
     size=12
     skewAmount=0
     typefaceFontName=Black
     outlineSettings {
      bApplyOutlineToDropShadows=false
      bMiteredCorners=false
      bSeparateFillAlpha=false
      outlineMaterial=null
      outlineSize=0
      outlineColor=<rgba 0 0 0 1>
     }
    }
   }
   slot {class=GridSlot, properties {row=1}}
  }
  UEFN_TextBlock_C_7 {
   widgetClassPath=&UEFN_TextBlock
   parent=ButtonsGridPanel_1
   widget=UEFN_TextBlock_C_7
   properties {
    text=.csv
    font {
     bForceMonospaced=false
     bMaterialIsStencil=false
     fontMaterial=null
     fontObject=&BurbankSmall
     letterSpacing=0
     monospacedWidth=1
     size=7.5
     skewAmount=0
     typefaceFontName=Black
     outlineSettings {
      bApplyOutlineToDropShadows=false
      bMiteredCorners=false
      bSeparateFillAlpha=false
      outlineMaterial=null
      outlineSize=0
      outlineColor=<rgba 0 0 0 1>
     }
    }
   }
   slot {class=GridSlot, properties {column=2, row=1, verticalAlignment=VAlign_Center}}
  }
  CsvPathTextBox {
   widgetClassPath=/Game/Valkyrie/UMG/UEFN_EditableTextBox
   bIsVariable=true
   parent=ButtonsGridPanel_1
   widget=CsvPathTextBox
   properties {widgetStyle {
     backgroundColor {colorUseRule=UseColor_Specified, specifiedColor=<rgba 1 1 1 1>}
     backgroundImageFocused {
      bIsDynamicallyLoaded=false
      drawAs=RoundedBox
      imageType=NoImage
      mirroring=NoMirror
      resourceName=None
      resourceObject=null
      tiling=NoTile
      imageSize=<0 0>
      margin=<ltrb 0 0 0 0>
      outlineSettings {
       bUseBrushTransparency=false
       roundingType=FixedRadius
       width=1
       color {colorUseRule=UseColor_ColorTable, specifiedColor=<rgba 0 0 0 0>}
       cornerRadii=<4 4 4 4>
      }
      tintColor {colorUseRule=UseColor_Specified, specifiedColor=<rgba 0 0.1248 0.4969 1>}
      uVRegion {bIsValid=false, max=<0 0>, min=<0 0>}
     }
     backgroundImageHovered {
      bIsDynamicallyLoaded=false
      drawAs=RoundedBox
      imageType=NoImage
      mirroring=NoMirror
      resourceName=None
      resourceObject=null
      tiling=NoTile
      imageSize=<0 0>
      margin=<ltrb 0 0 0 0>
      outlineSettings {
       bUseBrushTransparency=false
       roundingType=FixedRadius
       width=1
       color {colorUseRule=UseColor_ColorTable, specifiedColor=<rgba 0 0 0 0>}
       cornerRadii=<4 4 4 4>
      }
      tintColor {colorUseRule=UseColor_Specified, specifiedColor=<rgba 0.006 0.0561 0.162 1>}
      uVRegion {bIsValid=false, max=<0 0>, min=<0 0>}
     }
     backgroundImageNormal {
      bIsDynamicallyLoaded=false
      drawAs=RoundedBox
      imageType=NoImage
      mirroring=NoMirror
      resourceName=None
      resourceObject=null
      tiling=NoTile
      imageSize=<0 0>
      margin=<ltrb 0 0 0 0>
      outlineSettings {
       bUseBrushTransparency=false
       roundingType=FixedRadius
       width=1
       color {colorUseRule=UseColor_ColorTable, specifiedColor=<rgba 0 0 0 0>}
       cornerRadii=<4 4 4 4>
      }
      tintColor {colorUseRule=UseColor_Specified, specifiedColor=<rgba 0.0439 0.0809 0.1875 1>}
      uVRegion {bIsValid=false, max=<0 0>, min=<0 0>}
     }
     backgroundImageReadOnly {
      bIsDynamicallyLoaded=false
      drawAs=RoundedBox
      imageType=NoImage
      mirroring=NoMirror
      resourceName=None
      resourceObject=null
      tiling=NoTile
      imageSize=<0 0>
      margin=<ltrb 0 0 0 0>
      outlineSettings {
       bUseBrushTransparency=false
       roundingType=FixedRadius
       width=1
       color {colorUseRule=UseColor_ColorTable, specifiedColor=<rgba 0 0 0 0>}
       cornerRadii=<4 4 4 4>
      }
      tintColor {colorUseRule=UseColor_Specified, specifiedColor=<rgba 0.1946 0.1946 0.1946 1>}
      uVRegion {bIsValid=false, max=<0 0>, min=<0 0>}
     }
     focusedForegroundColor {colorUseRule=UseColor_Specified, specifiedColor=<rgba 1 1 1 1>}
     foregroundColor {colorUseRule=UseColor_Specified, specifiedColor=<rgba 1 1 1 1>}
     hScrollBarPadding=<ltrb 0 0 0 0>
     padding=<ltrb 8 3 8 4>
     readOnlyForegroundColor {colorUseRule=UseColor_Foreground, specifiedColor=<rgba 1 0 1 1>}
     scrollBarStyle {
      thickness=8
      draggedThumbImage {
       bIsDynamicallyLoaded=false
       drawAs=RoundedBox
       imageType=NoImage
       mirroring=NoMirror
       resourceName=None
       resourceObject=null
       tiling=NoTile
       imageSize=<0 0>
       margin=<ltrb 0 0 0 0>
       outlineSettings {
        bUseBrushTransparency=false
        roundingType=FixedRadius
        width=0
        color {colorUseRule=UseColor_Specified, specifiedColor=<rgba 0 0 0 0>}
        cornerRadii=<4 4 4 4>
       }
       tintColor {colorUseRule=UseColor_Specified, specifiedColor=<rgba 0.2159 0.2159 0.2159 1>}
       uVRegion {bIsValid=false, max=<0 0>, min=<0 0>}
      }
      horizontalBackgroundImage {
       bIsDynamicallyLoaded=false
       drawAs=NoDrawType
       imageType=NoImage
       mirroring=NoMirror
       resourceName=None
       resourceObject=null
       tiling=NoTile
       imageSize=<0 0>
       margin=<ltrb 0 0 0 0>
       outlineSettings {
        bUseBrushTransparency=false
        roundingType=HalfHeightRadius
        width=0
        color {colorUseRule=UseColor_Specified, specifiedColor=<rgba 0 0 0 0>}
        cornerRadii=<0 0 0 0>
       }
       tintColor {colorUseRule=UseColor_Specified, specifiedColor=<rgba 1 1 1 1>}
       uVRegion {bIsValid=false, max=<0 0>, min=<0 0>}
      }
      horizontalBottomSlotImage {
       bIsDynamicallyLoaded=false
       drawAs=NoDrawType
       imageType=NoImage
       mirroring=NoMirror
       resourceName=None
       resourceObject=null
       tiling=NoTile
       imageSize=<0 0>
       margin=<ltrb 0 0 0 0>
       outlineSettings {
        bUseBrushTransparency=false
        roundingType=HalfHeightRadius
        width=0
        color {colorUseRule=UseColor_Specified, specifiedColor=<rgba 0 0 0 0>}
        cornerRadii=<0 0 0 0>
       }
       tintColor {colorUseRule=UseColor_Specified, specifiedColor=<rgba 1 1 1 1>}
       uVRegion {bIsValid=false, max=<0 0>, min=<0 0>}
      }
      horizontalTopSlotImage {
       bIsDynamicallyLoaded=false
       drawAs=NoDrawType
       imageType=NoImage
       mirroring=NoMirror
       resourceName=None
       resourceObject=null
       tiling=NoTile
       imageSize=<0 0>
       margin=<ltrb 0 0 0 0>
       outlineSettings {
        bUseBrushTransparency=false
        roundingType=HalfHeightRadius
        width=0
        color {colorUseRule=UseColor_Specified, specifiedColor=<rgba 0 0 0 0>}
        cornerRadii=<0 0 0 0>
       }
       tintColor {colorUseRule=UseColor_Specified, specifiedColor=<rgba 1 1 1 1>}
       uVRegion {bIsValid=false, max=<0 0>, min=<0 0>}
      }
      hoveredThumbImage {
       bIsDynamicallyLoaded=false
       drawAs=RoundedBox
       imageType=NoImage
       mirroring=NoMirror
       resourceName=None
       resourceObject=null
       tiling=NoTile
       imageSize=<0 0>
       margin=<ltrb 0 0 0 0>
       outlineSettings {
        bUseBrushTransparency=false
        roundingType=FixedRadius
        width=0
        color {colorUseRule=UseColor_Specified, specifiedColor=<rgba 0 0 0 0>}
        cornerRadii=<4 4 4 4>
       }
       tintColor {colorUseRule=UseColor_Specified, specifiedColor=<rgba 0.2159 0.2159 0.2159 1>}
       uVRegion {bIsValid=false, max=<0 0>, min=<0 0>}
      }
      normalThumbImage {
       bIsDynamicallyLoaded=false
       drawAs=RoundedBox
       imageType=NoImage
       mirroring=NoMirror
       resourceName=None
       resourceObject=null
       tiling=NoTile
       imageSize=<0 0>
       margin=<ltrb 0 0 0 0>
       outlineSettings {
        bUseBrushTransparency=false
        roundingType=FixedRadius
        width=0
        color {colorUseRule=UseColor_Specified, specifiedColor=<rgba 0 0 0 0>}
        cornerRadii=<4 4 4 4>
       }
       tintColor {colorUseRule=UseColor_Specified, specifiedColor=<rgba 0.0953 0.0953 0.0953 1>}
       uVRegion {bIsValid=false, max=<0 0>, min=<0 0>}
      }
      verticalBackgroundImage {
       bIsDynamicallyLoaded=false
       drawAs=NoDrawType
       imageType=NoImage
       mirroring=NoMirror
       resourceName=None
       resourceObject=null
       tiling=NoTile
       imageSize=<0 0>
       margin=<ltrb 0 0 0 0>
       outlineSettings {
        bUseBrushTransparency=false
        roundingType=HalfHeightRadius
        width=0
        color {colorUseRule=UseColor_Specified, specifiedColor=<rgba 0 0 0 0>}
        cornerRadii=<0 0 0 0>
       }
       tintColor {colorUseRule=UseColor_Specified, specifiedColor=<rgba 1 1 1 1>}
       uVRegion {bIsValid=false, max=<0 0>, min=<0 0>}
      }
      verticalBottomSlotImage {
       bIsDynamicallyLoaded=false
       drawAs=NoDrawType
       imageType=NoImage
       mirroring=NoMirror
       resourceName=None
       resourceObject=null
       tiling=NoTile
       imageSize=<0 0>
       margin=<ltrb 0 0 0 0>
       outlineSettings {
        bUseBrushTransparency=false
        roundingType=HalfHeightRadius
        width=0
        color {colorUseRule=UseColor_Specified, specifiedColor=<rgba 0 0 0 0>}
        cornerRadii=<0 0 0 0>
       }
       tintColor {colorUseRule=UseColor_Specified, specifiedColor=<rgba 1 1 1 1>}
       uVRegion {bIsValid=false, max=<0 0>, min=<0 0>}
      }
      verticalTopSlotImage {
       bIsDynamicallyLoaded=false
       drawAs=NoDrawType
       imageType=NoImage
       mirroring=NoMirror
       resourceName=None
       resourceObject=null
       tiling=NoTile
       imageSize=<0 0>
       margin=<ltrb 0 0 0 0>
       outlineSettings {
        bUseBrushTransparency=false
        roundingType=HalfHeightRadius
        width=0
        color {colorUseRule=UseColor_Specified, specifiedColor=<rgba 0 0 0 0>}
        cornerRadii=<0 0 0 0>
       }
       tintColor {colorUseRule=UseColor_Specified, specifiedColor=<rgba 1 1 1 1>}
       uVRegion {bIsValid=false, max=<0 0>, min=<0 0>}
      }
     }
     textStyle {
      overflowPolicy=Clip
      transformPolicy=None
      colorAndOpacity {colorUseRule=UseColor_Foreground, specifiedColor=<rgba 1 0 1 1>}
      font {
       bForceMonospaced=false
       bMaterialIsStencil=false
       fontMaterial=null
       fontObject=null
       letterSpacing=0
       monospacedWidth=1
       size=7.5
       skewAmount=0
       typefaceFontName=Regular
       outlineSettings {
        bApplyOutlineToDropShadows=false
        bMiteredCorners=false
        bSeparateFillAlpha=false
        outlineMaterial=null
        outlineSize=0
        outlineColor=<rgba 0 0 0 1>
       }
      }
      highlightColor {colorUseRule=UseColor_ColorTable, specifiedColor=<rgba 0 0 0 0>}
      highlightShape {
       bIsDynamicallyLoaded=false
       drawAs=Image
       imageType=NoImage
       mirroring=NoMirror
       resourceName=None
       resourceObject=null
       tiling=NoTile
       imageSize=<0 0>
       margin=<ltrb 0 0 0 0>
       outlineSettings {
        bUseBrushTransparency=false
        roundingType=HalfHeightRadius
        width=0
        color {colorUseRule=UseColor_Specified, specifiedColor=<rgba 0 0 0 0>}
        cornerRadii=<0 0 0 0>
       }
       tintColor {colorUseRule=UseColor_ColorTable, specifiedColor=<rgba 0 0 0 0>}
       uVRegion {bIsValid=false, max=<0 0>, min=<0 0>}
      }
      selectedBackgroundColor {colorUseRule=UseColor_ColorTable, specifiedColor=<rgba 0 0 0 0>}
      shadowColorAndOpacity=<rgba 0 0 0 1>
      shadowOffset=<0 0>
      strikeBrush {
       bIsDynamicallyLoaded=false
       drawAs=Image
       imageType=NoImage
       mirroring=NoMirror
       resourceName=None
       resourceObject=null
       tiling=NoTile
       imageSize=<32 32>
       margin=<ltrb 0 0 0 0>
       outlineSettings {
        bUseBrushTransparency=false
        roundingType=HalfHeightRadius
        width=0
        color {colorUseRule=UseColor_Specified, specifiedColor=<rgba 0 0 0 0>}
        cornerRadii=<0 0 0 0>
       }
       tintColor {colorUseRule=UseColor_Specified, specifiedColor=<rgba 1 1 1 1>}
       uVRegion {bIsValid=false, max=<0 0>, min=<0 0>}
      }
      underlineBrush {
       bIsDynamicallyLoaded=false
       drawAs=Image
       imageType=NoImage
       mirroring=NoMirror
       resourceName=None
       resourceObject=null
       tiling=NoTile
       imageSize=<32 32>
       margin=<ltrb 0 0 0 0>
       outlineSettings {
        bUseBrushTransparency=false
        roundingType=HalfHeightRadius
        width=0
        color {colorUseRule=UseColor_Specified, specifiedColor=<rgba 0 0 0 0>}
        cornerRadii=<0 0 0 0>
       }
       tintColor {colorUseRule=UseColor_Specified, specifiedColor=<rgba 1 1 1 1>}
       uVRegion {bIsValid=false, max=<0 0>, min=<0 0>}
      }
     }
     vScrollBarPadding=<ltrb 0 0 0 0>
    }}
   slot {class=GridSlot, properties {column=1, columnSpan=2, verticalAlignment=VAlign_Bottom}}
  }
  CsvNameTextBox {
   widgetClassPath=&UEFN_EditableTextBox
   bIsVariable=true
   parent=ButtonsGridPanel_1
   widget=CsvNameTextBox
   properties {
    text=Filename
    widgetStyle {
     backgroundColor {colorUseRule=UseColor_Specified, specifiedColor=<rgba 1 1 1 1>}
     backgroundImageFocused {
      bIsDynamicallyLoaded=false
      drawAs=RoundedBox
      imageType=NoImage
      mirroring=NoMirror
      resourceName=None
      resourceObject=null
      tiling=NoTile
      imageSize=<0 0>
      margin=<ltrb 0 0 0 0>
      outlineSettings {
       bUseBrushTransparency=false
       roundingType=FixedRadius
       width=1
       color {colorUseRule=UseColor_ColorTable, specifiedColor=<rgba 0 0 0 0>}
       cornerRadii=<4 4 4 4>
      }
      tintColor {colorUseRule=UseColor_Specified, specifiedColor=<rgba 0 0.1248 0.4969 1>}
      uVRegion {bIsValid=false, max=<0 0>, min=<0 0>}
     }
     backgroundImageHovered {
      bIsDynamicallyLoaded=false
      drawAs=RoundedBox
      imageType=NoImage
      mirroring=NoMirror
      resourceName=None
      resourceObject=null
      tiling=NoTile
      imageSize=<0 0>
      margin=<ltrb 0 0 0 0>
      outlineSettings {
       bUseBrushTransparency=false
       roundingType=FixedRadius
       width=1
       color {colorUseRule=UseColor_ColorTable, specifiedColor=<rgba 0 0 0 0>}
       cornerRadii=<4 4 4 4>
      }
      tintColor {colorUseRule=UseColor_Specified, specifiedColor=<rgba 0.006 0.0561 0.162 1>}
      uVRegion {bIsValid=false, max=<0 0>, min=<0 0>}
     }
     backgroundImageNormal {
      bIsDynamicallyLoaded=false
      drawAs=RoundedBox
      imageType=NoImage
      mirroring=NoMirror
      resourceName=None
      resourceObject=null
      tiling=NoTile
      imageSize=<0 0>
      margin=<ltrb 0 0 0 0>
      outlineSettings {
       bUseBrushTransparency=false
       roundingType=FixedRadius
       width=1
       color {colorUseRule=UseColor_ColorTable, specifiedColor=<rgba 0 0 0 0>}
       cornerRadii=<4 4 4 4>
      }
      tintColor {colorUseRule=UseColor_Specified, specifiedColor=<rgba 0.0439 0.0809 0.1875 1>}
      uVRegion {bIsValid=false, max=<0 0>, min=<0 0>}
     }
     backgroundImageReadOnly {
      bIsDynamicallyLoaded=false
      drawAs=RoundedBox
      imageType=NoImage
      mirroring=NoMirror
      resourceName=None
      resourceObject=null
      tiling=NoTile
      imageSize=<0 0>
      margin=<ltrb 0 0 0 0>
      outlineSettings {
       bUseBrushTransparency=false
       roundingType=FixedRadius
       width=1
       color {colorUseRule=UseColor_ColorTable, specifiedColor=<rgba 0 0 0 0>}
       cornerRadii=<4 4 4 4>
      }
      tintColor {colorUseRule=UseColor_Specified, specifiedColor=<rgba 0.1946 0.1946 0.1946 1>}
      uVRegion {bIsValid=false, max=<0 0>, min=<0 0>}
     }
     focusedForegroundColor {colorUseRule=UseColor_Specified, specifiedColor=<rgba 1 1 1 1>}
     foregroundColor {colorUseRule=UseColor_Specified, specifiedColor=<rgba 1 1 1 1>}
     hScrollBarPadding=<ltrb 0 0 0 0>
     padding=<ltrb 8 3 8 4>
     readOnlyForegroundColor {colorUseRule=UseColor_Foreground, specifiedColor=<rgba 1 0 1 1>}
     scrollBarStyle {
      thickness=8
      draggedThumbImage {
       bIsDynamicallyLoaded=false
       drawAs=RoundedBox
       imageType=NoImage
       mirroring=NoMirror
       resourceName=None
       resourceObject=null
       tiling=NoTile
       imageSize=<0 0>
       margin=<ltrb 0 0 0 0>
       outlineSettings {
        bUseBrushTransparency=false
        roundingType=FixedRadius
        width=0
        color {colorUseRule=UseColor_Specified, specifiedColor=<rgba 0 0 0 0>}
        cornerRadii=<4 4 4 4>
       }
       tintColor {colorUseRule=UseColor_Specified, specifiedColor=<rgba 0.2159 0.2159 0.2159 1>}
       uVRegion {bIsValid=false, max=<0 0>, min=<0 0>}
      }
      horizontalBackgroundImage {
       bIsDynamicallyLoaded=false
       drawAs=NoDrawType
       imageType=NoImage
       mirroring=NoMirror
       resourceName=None
       resourceObject=null
       tiling=NoTile
       imageSize=<0 0>
       margin=<ltrb 0 0 0 0>
       outlineSettings {
        bUseBrushTransparency=false
        roundingType=HalfHeightRadius
        width=0
        color {colorUseRule=UseColor_Specified, specifiedColor=<rgba 0 0 0 0>}
        cornerRadii=<0 0 0 0>
       }
       tintColor {colorUseRule=UseColor_Specified, specifiedColor=<rgba 1 1 1 1>}
       uVRegion {bIsValid=false, max=<0 0>, min=<0 0>}
      }
      horizontalBottomSlotImage {
       bIsDynamicallyLoaded=false
       drawAs=NoDrawType
       imageType=NoImage
       mirroring=NoMirror
       resourceName=None
       resourceObject=null
       tiling=NoTile
       imageSize=<0 0>
       margin=<ltrb 0 0 0 0>
       outlineSettings {
        bUseBrushTransparency=false
        roundingType=HalfHeightRadius
        width=0
        color {colorUseRule=UseColor_Specified, specifiedColor=<rgba 0 0 0 0>}
        cornerRadii=<0 0 0 0>
       }
       tintColor {colorUseRule=UseColor_Specified, specifiedColor=<rgba 1 1 1 1>}
       uVRegion {bIsValid=false, max=<0 0>, min=<0 0>}
      }
      horizontalTopSlotImage {
       bIsDynamicallyLoaded=false
       drawAs=NoDrawType
       imageType=NoImage
       mirroring=NoMirror
       resourceName=None
       resourceObject=null
       tiling=NoTile
       imageSize=<0 0>
       margin=<ltrb 0 0 0 0>
       outlineSettings {
        bUseBrushTransparency=false
        roundingType=HalfHeightRadius
        width=0
        color {colorUseRule=UseColor_Specified, specifiedColor=<rgba 0 0 0 0>}
        cornerRadii=<0 0 0 0>
       }
       tintColor {colorUseRule=UseColor_Specified, specifiedColor=<rgba 1 1 1 1>}
       uVRegion {bIsValid=false, max=<0 0>, min=<0 0>}
      }
      hoveredThumbImage {
       bIsDynamicallyLoaded=false
       drawAs=RoundedBox
       imageType=NoImage
       mirroring=NoMirror
       resourceName=None
       resourceObject=null
       tiling=NoTile
       imageSize=<0 0>
       margin=<ltrb 0 0 0 0>
       outlineSettings {
        bUseBrushTransparency=false
        roundingType=FixedRadius
        width=0
        color {colorUseRule=UseColor_Specified, specifiedColor=<rgba 0 0 0 0>}
        cornerRadii=<4 4 4 4>
       }
       tintColor {colorUseRule=UseColor_Specified, specifiedColor=<rgba 0.2159 0.2159 0.2159 1>}
       uVRegion {bIsValid=false, max=<0 0>, min=<0 0>}
      }
      normalThumbImage {
       bIsDynamicallyLoaded=false
       drawAs=RoundedBox
       imageType=NoImage
       mirroring=NoMirror
       resourceName=None
       resourceObject=null
       tiling=NoTile
       imageSize=<0 0>
       margin=<ltrb 0 0 0 0>
       outlineSettings {
        bUseBrushTransparency=false
        roundingType=FixedRadius
        width=0
        color {colorUseRule=UseColor_Specified, specifiedColor=<rgba 0 0 0 0>}
        cornerRadii=<4 4 4 4>
       }
       tintColor {colorUseRule=UseColor_Specified, specifiedColor=<rgba 0.0953 0.0953 0.0953 1>}
       uVRegion {bIsValid=false, max=<0 0>, min=<0 0>}
      }
      verticalBackgroundImage {
       bIsDynamicallyLoaded=false
       drawAs=NoDrawType
       imageType=NoImage
       mirroring=NoMirror
       resourceName=None
       resourceObject=null
       tiling=NoTile
       imageSize=<0 0>
       margin=<ltrb 0 0 0 0>
       outlineSettings {
        bUseBrushTransparency=false
        roundingType=HalfHeightRadius
        width=0
        color {colorUseRule=UseColor_Specified, specifiedColor=<rgba 0 0 0 0>}
        cornerRadii=<0 0 0 0>
       }
       tintColor {colorUseRule=UseColor_Specified, specifiedColor=<rgba 1 1 1 1>}
       uVRegion {bIsValid=false, max=<0 0>, min=<0 0>}
      }
      verticalBottomSlotImage {
       bIsDynamicallyLoaded=false
       drawAs=NoDrawType
       imageType=NoImage
       mirroring=NoMirror
       resourceName=None
       resourceObject=null
       tiling=NoTile
       imageSize=<0 0>
       margin=<ltrb 0 0 0 0>
       outlineSettings {
        bUseBrushTransparency=false
        roundingType=HalfHeightRadius
        width=0
        color {colorUseRule=UseColor_Specified, specifiedColor=<rgba 0 0 0 0>}
        cornerRadii=<0 0 0 0>
       }
       tintColor {colorUseRule=UseColor_Specified, specifiedColor=<rgba 1 1 1 1>}
       uVRegion {bIsValid=false, max=<0 0>, min=<0 0>}
      }
      verticalTopSlotImage {
       bIsDynamicallyLoaded=false
       drawAs=NoDrawType
       imageType=NoImage
       mirroring=NoMirror
       resourceName=None
       resourceObject=null
       tiling=NoTile
       imageSize=<0 0>
       margin=<ltrb 0 0 0 0>
       outlineSettings {
        bUseBrushTransparency=false
        roundingType=HalfHeightRadius
        width=0
        color {colorUseRule=UseColor_Specified, specifiedColor=<rgba 0 0 0 0>}
        cornerRadii=<0 0 0 0>
       }
       tintColor {colorUseRule=UseColor_Specified, specifiedColor=<rgba 1 1 1 1>}
       uVRegion {bIsValid=false, max=<0 0>, min=<0 0>}
      }
     }
     textStyle {
      overflowPolicy=Clip
      transformPolicy=None
      colorAndOpacity {colorUseRule=UseColor_Foreground, specifiedColor=<rgba 1 0 1 1>}
      font {
       bForceMonospaced=false
       bMaterialIsStencil=false
       fontMaterial=null
       fontObject=null
       letterSpacing=0
       monospacedWidth=1
       size=7.5
       skewAmount=0
       typefaceFontName=Regular
       outlineSettings {
        bApplyOutlineToDropShadows=false
        bMiteredCorners=false
        bSeparateFillAlpha=false
        outlineMaterial=null
        outlineSize=0
        outlineColor=<rgba 0 0 0 1>
       }
      }
      highlightColor {colorUseRule=UseColor_ColorTable, specifiedColor=<rgba 0 0 0 0>}
      highlightShape {
       bIsDynamicallyLoaded=false
       drawAs=Image
       imageType=NoImage
       mirroring=NoMirror
       resourceName=None
       resourceObject=null
       tiling=NoTile
       imageSize=<0 0>
       margin=<ltrb 0 0 0 0>
       outlineSettings {
        bUseBrushTransparency=false
        roundingType=HalfHeightRadius
        width=0
        color {colorUseRule=UseColor_Specified, specifiedColor=<rgba 0 0 0 0>}
        cornerRadii=<0 0 0 0>
       }
       tintColor {colorUseRule=UseColor_ColorTable, specifiedColor=<rgba 0 0 0 0>}
       uVRegion {bIsValid=false, max=<0 0>, min=<0 0>}
      }
      selectedBackgroundColor {colorUseRule=UseColor_ColorTable, specifiedColor=<rgba 0 0 0 0>}
      shadowColorAndOpacity=<rgba 0 0 0 1>
      shadowOffset=<0 0>
      strikeBrush {
       bIsDynamicallyLoaded=false
       drawAs=Image
       imageType=NoImage
       mirroring=NoMirror
       resourceName=None
       resourceObject=null
       tiling=NoTile
       imageSize=<32 32>
       margin=<ltrb 0 0 0 0>
       outlineSettings {
        bUseBrushTransparency=false
        roundingType=HalfHeightRadius
        width=0
        color {colorUseRule=UseColor_Specified, specifiedColor=<rgba 0 0 0 0>}
        cornerRadii=<0 0 0 0>
       }
       tintColor {colorUseRule=UseColor_Specified, specifiedColor=<rgba 1 1 1 1>}
       uVRegion {bIsValid=false, max=<0 0>, min=<0 0>}
      }
      underlineBrush {
       bIsDynamicallyLoaded=false
       drawAs=Image
       imageType=NoImage
       mirroring=NoMirror
       resourceName=None
       resourceObject=null
       tiling=NoTile
       imageSize=<32 32>
       margin=<ltrb 0 0 0 0>
       outlineSettings {
        bUseBrushTransparency=false
        roundingType=HalfHeightRadius
        width=0
        color {colorUseRule=UseColor_Specified, specifiedColor=<rgba 0 0 0 0>}
        cornerRadii=<0 0 0 0>
       }
       tintColor {colorUseRule=UseColor_Specified, specifiedColor=<rgba 1 1 1 1>}
       uVRegion {bIsValid=false, max=<0 0>, min=<0 0>}
      }
     }
     vScrollBarPadding=<ltrb 0 0 0 0>
    }
   }
   slot {class=GridSlot, properties {column=1, row=1, verticalAlignment=VAlign_Bottom}}
  }
  AnalyzerSettingsHeaderSizeBox {
   widgetClassPath=&SizeBox
   parent=MainGridPanel
   widget=AnalyzerSettingsHeaderSizeBox
   slot {class=GridSlot, properties {column=1, row=4}}
  }
  AnalyzerSettingsHeaderBorder {
   widgetClassPath=&Border
   parent=AnalyzerSettingsHeaderSizeBox
   widget=AnalyzerSettingsHeaderBorder
   properties {horizontalAlignment=HAlign_Center, brushColor=<rgba 0 0.0339 0.0833 1>, padding=<ltrb 5 5 0 4>}
   slot {class=SizeBoxSlot}
  }
  Header_01 {
   widgetClassPath=&TextBlock
   parent=AnalyzerSettingsHeaderBorder
   widget=Header_01
   properties {
    text=Analyzer Settings
    font {
     bForceMonospaced=false
     bMaterialIsStencil=false
     fontMaterial=null
     fontObject=/LauncherSocial/BrutalType
     letterSpacing=100
     monospacedWidth=1
     size=21
     skewAmount=0
     typefaceFontName=Light
     outlineSettings {
      bApplyOutlineToDropShadows=false
      bMiteredCorners=false
      bSeparateFillAlpha=false
      outlineMaterial=null
      outlineSize=0
      outlineColor=<rgba 0 0 0 1>
     }
    }
   }
   slot {class=BorderSlot, properties {horizontalAlignment=HAlign_Center, padding=<ltrb 5 5 0 4>}}
  }
  AnalyzerSettingsSizeBox {
   widgetClassPath=&SizeBox
   parent=MainGridPanel
   widget=AnalyzerSettingsSizeBox
   properties {bOverride_HeightOverride=true, bOverride_WidthOverride=true, heightOverride=100, widthOverride=400}
   slot {class=GridSlot, properties {column=1, layer=10, row=5, rowSpan=3}}
  }
  AnalyzerSettingsDetailsView {
   widgetClassPath=&DetailsView
   parent=AnalyzerSettingsSizeBox
   widget=AnalyzerSettingsDetailsView
   properties {columnWidth=0.5184, propertiesToShow [LoudnessAnalyzerNRTSettings]}
   slot {class=SizeBoxSlot}
  }
  CurrentAnalyzerSettingsSizeBox {
   widgetClassPath=&SizeBox
   parent=MainGridPanel
   widget=CurrentAnalyzerSettingsSizeBox
   properties {bOverride_WidthOverride=true, widthOverride=400}
   slot {class=GridSlot, properties {column=1, row=8}}
  }
  CurrentAnalyzerSettingsBorder {
   widgetClassPath=&Border
   parent=CurrentAnalyzerSettingsSizeBox
   widget=CurrentAnalyzerSettingsBorder
   properties {brushColor=<rgba 0.0103 0.0103 0.0103 1>, padding=<ltrb 0 0 0 0>}
   slot {class=SizeBoxSlot}
  }
  CurrentSettingsGridPanel {
   widgetClassPath=&GridPanel
   parent=CurrentAnalyzerSettingsBorder
   widget=CurrentSettingsGridPanel
   slot {class=BorderSlot, properties {padding=<ltrb 0 0 0 0>}}
  }
  TextBlock_458 {
   widgetClassPath=&TextBlock
   parent=CurrentSettingsGridPanel
   widget=TextBlock_458
   properties {
    text=Current Settings:
    colorAndOpacity {colorUseRule=UseColor_Specified, specifiedColor=<rgba 0.5776 0.305 0.5776 1>}
    font {
     bForceMonospaced=false
     bMaterialIsStencil=false
     fontMaterial=null
     fontObject=null
     letterSpacing=0
     monospacedWidth=1
     size=12
     skewAmount=0
     typefaceFontName=74 Regular
     outlineSettings {
      bApplyOutlineToDropShadows=false
      bMiteredCorners=false
      bSeparateFillAlpha=false
      outlineMaterial=null
      outlineSize=0
      outlineColor=<rgba 0 0 0 1>
     }
    }
   }
   slot {class=GridSlot, properties {column=1, columnSpan=2, row=1, padding=<ltrb 0 5 0 0>}}
  }
  Spacer_3 {
   widgetClassPath=&Spacer
   parent=CurrentSettingsGridPanel
   widget=Spacer_3
   slot {class=GridSlot, properties {padding=<ltrb 0 0 10 2>}}
  }
  Spacer_5 {
   widgetClassPath=&Spacer
   parent=CurrentSettingsGridPanel
   widget=Spacer_5
   properties {size=<1 33.2247>}
   slot {class=GridSlot, properties {column=1, row=1, padding=<ltrb 0 0 10 2>}}
  }
  Spacer_4 {
   widgetClassPath=&Spacer
   parent=CurrentSettingsGridPanel
   widget=Spacer_4
   properties {size=<156.5327 1>}
   slot {class=GridSlot, properties {column=1, row=2}}
  }
  TextBlock_1 {
   widgetClassPath=&TextBlock
   parent=CurrentSettingsGridPanel
   widget=TextBlock_1
   properties {
    text=Analysis Period:
    colorAndOpacity {colorUseRule=UseColor_Specified, specifiedColor=<rgba 0.2798 0.2893 0.401 1>}
    font {
     bForceMonospaced=false
     bMaterialIsStencil=false
     fontMaterial=null
     fontObject=null
     letterSpacing=0
     monospacedWidth=1
     size=9
     skewAmount=0
     typefaceFontName=74 Regular
     outlineSettings {
      bApplyOutlineToDropShadows=false
      bMiteredCorners=false
      bSeparateFillAlpha=false
      outlineMaterial=null
      outlineSize=0
      outlineColor=<rgba 0 0 0 1>
     }
    }
   }
   slot {class=GridSlot, properties {column=1, row=2}}
  }
  TextBlock_2 {
   widgetClassPath=&TextBlock
   parent=CurrentSettingsGridPanel
   widget=TextBlock_2
   properties {
    text=Min Frequency:
    colorAndOpacity {colorUseRule=UseColor_Specified, specifiedColor=<rgba 0.2798 0.2893 0.401 1>}
    font {
     bForceMonospaced=false
     bMaterialIsStencil=false
     fontMaterial=null
     fontObject=null
     letterSpacing=0
     monospacedWidth=1
     size=9
     skewAmount=0
     typefaceFontName=74 Regular
     outlineSettings {
      bApplyOutlineToDropShadows=false
      bMiteredCorners=false
      bSeparateFillAlpha=false
      outlineMaterial=null
      outlineSize=0
      outlineColor=<rgba 0 0 0 1>
     }
    }
   }
   slot {class=GridSlot, properties {column=1, row=3}}
  }
  TextBlock_3 {
   widgetClassPath=&TextBlock
   parent=CurrentSettingsGridPanel
   widget=TextBlock_3
   properties {
    text=Max Frequency:
    colorAndOpacity {colorUseRule=UseColor_Specified, specifiedColor=<rgba 0.2798 0.2893 0.401 1>}
    font {
     bForceMonospaced=false
     bMaterialIsStencil=false
     fontMaterial=null
     fontObject=null
     letterSpacing=0
     monospacedWidth=1
     size=9
     skewAmount=0
     typefaceFontName=74 Regular
     outlineSettings {
      bApplyOutlineToDropShadows=false
      bMiteredCorners=false
      bSeparateFillAlpha=false
      outlineMaterial=null
      outlineSize=0
      outlineColor=<rgba 0 0 0 1>
     }
    }
   }
   slot {class=GridSlot, properties {column=1, row=4}}
  }
  TextBlock_7 {
   widgetClassPath=&TextBlock
   parent=CurrentSettingsGridPanel
   widget=TextBlock_7
   properties {
    text=Curve Type:
    colorAndOpacity {colorUseRule=UseColor_Specified, specifiedColor=<rgba 0.2798 0.2893 0.401 1>}
    font {
     bForceMonospaced=false
     bMaterialIsStencil=false
     fontMaterial=null
     fontObject=null
     letterSpacing=0
     monospacedWidth=1
     size=9
     skewAmount=0
     typefaceFontName=74 Regular
     outlineSettings {
      bApplyOutlineToDropShadows=false
      bMiteredCorners=false
      bSeparateFillAlpha=false
      outlineMaterial=null
      outlineSize=0
      outlineColor=<rgba 0 0 0 1>
     }
    }
   }
   slot {class=GridSlot, properties {column=1, row=5}}
  }
  TextBlock_9 {
   widgetClassPath=&TextBlock
   parent=CurrentSettingsGridPanel
   widget=TextBlock_9
   properties {
    text=Noise Floor:
    colorAndOpacity {colorUseRule=UseColor_Specified, specifiedColor=<rgba 0.2798 0.2893 0.401 1>}
    font {
     bForceMonospaced=false
     bMaterialIsStencil=false
     fontMaterial=null
     fontObject=null
     letterSpacing=0
     monospacedWidth=1
     size=9
     skewAmount=0
     typefaceFontName=74 Regular
     outlineSettings {
      bApplyOutlineToDropShadows=false
      bMiteredCorners=false
      bSeparateFillAlpha=false
      outlineMaterial=null
      outlineSize=0
      outlineColor=<rgba 0 0 0 1>
     }
    }
   }
   slot {class=GridSlot, properties {column=1, row=6, padding=<ltrb 0 0 0 15>}}
  }
  Value_AnalysisPeriod {
   widgetClassPath=&TextBlock
   bIsVariable=true
   parent=CurrentSettingsGridPanel
   widget=Value_AnalysisPeriod
   properties {
    text=0.00
    colorAndOpacity {colorUseRule=UseColor_Specified, specifiedColor=<rgba 0.2798 0.2893 0.401 1>}
    font {
     bForceMonospaced=false
     bMaterialIsStencil=false
     fontMaterial=null
     fontObject=null
     letterSpacing=0
     monospacedWidth=1
     size=9
     skewAmount=0
     typefaceFontName=74 Regular
     outlineSettings {
      bApplyOutlineToDropShadows=false
      bMiteredCorners=false
      bSeparateFillAlpha=false
      outlineMaterial=null
      outlineSize=0
      outlineColor=<rgba 0 0 0 1>
     }
    }
   }
   slot {class=GridSlot, properties {column=2, horizontalAlignment=HAlign_Left, row=2}}
  }
  Value_MinFrequency {
   widgetClassPath=&TextBlock
   bIsVariable=true
   parent=CurrentSettingsGridPanel
   widget=Value_MinFrequency
   properties {
    text=0.00
    colorAndOpacity {colorUseRule=UseColor_Specified, specifiedColor=<rgba 0.2798 0.2893 0.401 1>}
    font {
     bForceMonospaced=false
     bMaterialIsStencil=false
     fontMaterial=null
     fontObject=null
     letterSpacing=0
     monospacedWidth=1
     size=9
     skewAmount=0
     typefaceFontName=74 Regular
     outlineSettings {
      bApplyOutlineToDropShadows=false
      bMiteredCorners=false
      bSeparateFillAlpha=false
      outlineMaterial=null
      outlineSize=0
      outlineColor=<rgba 0 0 0 1>
     }
    }
   }
   slot {class=GridSlot, properties {column=2, horizontalAlignment=HAlign_Left, row=3}}
  }
  Value_MaxFrequency {
   widgetClassPath=&TextBlock
   bIsVariable=true
   parent=CurrentSettingsGridPanel
   widget=Value_MaxFrequency
   properties {
    text=0.00
    colorAndOpacity {colorUseRule=UseColor_Specified, specifiedColor=<rgba 0.2798 0.2893 0.401 1>}
    font {
     bForceMonospaced=false
     bMaterialIsStencil=false
     fontMaterial=null
     fontObject=null
     letterSpacing=0
     monospacedWidth=1
     size=9
     skewAmount=0
     typefaceFontName=74 Regular
     outlineSettings {
      bApplyOutlineToDropShadows=false
      bMiteredCorners=false
      bSeparateFillAlpha=false
      outlineMaterial=null
      outlineSize=0
      outlineColor=<rgba 0 0 0 1>
     }
    }
   }
   slot {class=GridSlot, properties {column=2, horizontalAlignment=HAlign_Left, row=4}}
  }
  Value_CurveType {
   widgetClassPath=&TextBlock
   bIsVariable=true
   parent=CurrentSettingsGridPanel
   widget=Value_CurveType
   properties {
    text=D
    colorAndOpacity {colorUseRule=UseColor_Specified, specifiedColor=<rgba 0.2798 0.2893 0.401 1>}
    font {
     bForceMonospaced=false
     bMaterialIsStencil=false
     fontMaterial=null
     fontObject=null
     letterSpacing=0
     monospacedWidth=1
     size=9
     skewAmount=0
     typefaceFontName=74 Regular
     outlineSettings {
      bApplyOutlineToDropShadows=false
      bMiteredCorners=false
      bSeparateFillAlpha=false
      outlineMaterial=null
      outlineSize=0
      outlineColor=<rgba 0 0 0 1>
     }
    }
   }
   slot {class=GridSlot, properties {column=2, horizontalAlignment=HAlign_Left, row=5}}
  }
  Value_NoiseFloor {
   widgetClassPath=&TextBlock
   bIsVariable=true
   parent=CurrentSettingsGridPanel
   widget=Value_NoiseFloor
   properties {
    text=-60.00
    colorAndOpacity {colorUseRule=UseColor_Specified, specifiedColor=<rgba 0.2798 0.2893 0.401 1>}
    font {
     bForceMonospaced=false
     bMaterialIsStencil=false
     fontMaterial=null
     fontObject=null
     letterSpacing=0
     monospacedWidth=1
     size=9
     skewAmount=0
     typefaceFontName=74 Regular
     outlineSettings {
      bApplyOutlineToDropShadows=false
      bMiteredCorners=false
      bSeparateFillAlpha=false
      outlineMaterial=null
      outlineSize=0
      outlineColor=<rgba 0 0 0 1>
     }
    }
   }
   slot {class=GridSlot, properties {column=2, horizontalAlignment=HAlign_Left, row=6}}
  }
  TextBlock_11 {
   widgetClassPath=&TextBlock
   parent=CurrentSettingsGridPanel
   widget=TextBlock_11
   properties {
    text=Seconds
    colorAndOpacity {colorUseRule=UseColor_Specified, specifiedColor=<rgba 0.2798 0.2893 0.401 1>}
    font {
     bForceMonospaced=false
     bMaterialIsStencil=false
     fontMaterial=null
     fontObject=null
     letterSpacing=0
     monospacedWidth=1
     size=9
     skewAmount=0
     typefaceFontName=74 Regular
     outlineSettings {
      bApplyOutlineToDropShadows=false
      bMiteredCorners=false
      bSeparateFillAlpha=false
      outlineMaterial=null
      outlineSize=0
      outlineColor=<rgba 0 0 0 1>
     }
    }
   }
   slot {class=GridSlot, properties {column=3, horizontalAlignment=HAlign_Left, row=2}}
  }
  TextBlock_12 {
   widgetClassPath=&TextBlock
   parent=CurrentSettingsGridPanel
   widget=TextBlock_12
   properties {
    text=Hz
    colorAndOpacity {colorUseRule=UseColor_Specified, specifiedColor=<rgba 0.2798 0.2893 0.401 1>}
    font {
     bForceMonospaced=false
     bMaterialIsStencil=false
     fontMaterial=null
     fontObject=null
     letterSpacing=0
     monospacedWidth=1
     size=9
     skewAmount=0
     typefaceFontName=74 Regular
     outlineSettings {
      bApplyOutlineToDropShadows=false
      bMiteredCorners=false
      bSeparateFillAlpha=false
      outlineMaterial=null
      outlineSize=0
      outlineColor=<rgba 0 0 0 1>
     }
    }
   }
   slot {class=GridSlot, properties {column=3, horizontalAlignment=HAlign_Left, row=3}}
  }
  TextBlock_13 {
   widgetClassPath=&TextBlock
   parent=CurrentSettingsGridPanel
   widget=TextBlock_13
   properties {
    text=Hz
    colorAndOpacity {colorUseRule=UseColor_Specified, specifiedColor=<rgba 0.2798 0.2893 0.401 1>}
    font {
     bForceMonospaced=false
     bMaterialIsStencil=false
     fontMaterial=null
     fontObject=null
     letterSpacing=0
     monospacedWidth=1
     size=9
     skewAmount=0
     typefaceFontName=74 Regular
     outlineSettings {
      bApplyOutlineToDropShadows=false
      bMiteredCorners=false
      bSeparateFillAlpha=false
      outlineMaterial=null
      outlineSize=0
      outlineColor=<rgba 0 0 0 1>
     }
    }
   }
   slot {class=GridSlot, properties {column=3, horizontalAlignment=HAlign_Left, row=4}}
  }
  TextBlock_14 {
   widgetClassPath=&TextBlock
   parent=CurrentSettingsGridPanel
   widget=TextBlock_14
   properties {
    text=dB
    colorAndOpacity {colorUseRule=UseColor_Specified, specifiedColor=<rgba 0.2798 0.2893 0.401 1>}
    font {
     bForceMonospaced=false
     bMaterialIsStencil=false
     fontMaterial=null
     fontObject=null
     letterSpacing=0
     monospacedWidth=1
     size=9
     skewAmount=0
     typefaceFontName=74 Regular
     outlineSettings {
      bApplyOutlineToDropShadows=false
      bMiteredCorners=false
      bSeparateFillAlpha=false
      outlineMaterial=null
      outlineSize=0
      outlineColor=<rgba 0 0 0 1>
     }
    }
   }
   slot {class=GridSlot, properties {column=3, horizontalAlignment=HAlign_Left, row=6}}
  }
  Spacer_CurrentAnalyzerSettings_Bottom {
   widgetClassPath=&Spacer
   parent=MainGridPanel
   widget=Spacer_CurrentAnalyzerSettings_Bottom
   properties {size=<1 2.3913>}
   slot {class=GridSlot, properties {row=9, padding=<ltrb 0 0 20 20>}}
  }
  Spacer_Buttons_Top {
   widgetClassPath=&Spacer
   parent=MainGridPanel
   widget=Spacer_Buttons_Top
   properties {size=<1 -11.3366>}
   slot {class=GridSlot, properties {row=15, padding=<ltrb 0 0 20 20>}}
  }
  Spacer_Buttons_Bottom {
   widgetClassPath=&Spacer
   parent=MainGridPanel
   widget=Spacer_Buttons_Bottom
   properties {size=<1 12.4112>}
   slot {class=GridSlot, properties {row=17, padding=<ltrb 0 0 20 20>}}
  }
  HeaderBorder_02 {
   widgetClassPath=&Border
   parent=MainGridPanel
   widget=HeaderBorder_02
   properties {horizontalAlignment=HAlign_Center, verticalAlignment=VAlign_Bottom, brushColor=<rgba 0 0.0339 0.0833 1>, padding=<ltrb 5 5 0 4>}
   slot {class=GridSlot, properties {column=1, row=10}}
  }
  TextBlock_312 {
   widgetClassPath=&TextBlock
   parent=HeaderBorder_02
   widget=TextBlock_312
   properties {
    text=Assets to Analyze
    font {
     bForceMonospaced=false
     bMaterialIsStencil=false
     fontMaterial=null
     fontObject=&BrutalType
     letterSpacing=100
     monospacedWidth=1
     size=21
     skewAmount=0
     typefaceFontName=Light
     outlineSettings {
      bApplyOutlineToDropShadows=false
      bMiteredCorners=false
      bSeparateFillAlpha=false
      outlineMaterial=null
      outlineSize=0
      outlineColor=<rgba 0 0 0 1>
     }
    }
   }
   slot {class=BorderSlot, properties {horizontalAlignment=HAlign_Center, verticalAlignment=VAlign_Bottom, padding=<ltrb 5 5 0 4>}}
  }
  SoundWaveDetailsViewSizeBox {
   widgetClassPath=&SizeBox
   parent=MainGridPanel
   widget=SoundWaveDetailsViewSizeBox
   properties {bOverride_HeightOverride=true, bOverride_WidthOverride=true, heightOverride=400, widthOverride=400}
   slot {class=GridSlot, properties {column=1, layer=10, row=13, rowSpan=2}}
  }
  SoundWaveDetailsView {
   widgetClassPath=&DetailsView
   parent=SoundWaveDetailsViewSizeBox
   widget=SoundWaveDetailsView
   properties {bAllowFiltering=true, columnWidth=0.5976, propertiesToShow [SoundWaves]}
   slot {class=SizeBoxSlot}
  }
  HeaderBorder_03 {
   widgetClassPath=&Border
   parent=MainGridPanel
   widget=HeaderBorder_03
   properties {horizontalAlignment=HAlign_Center, verticalAlignment=VAlign_Center, brushColor=<rgba 0 0.0339 0.0833 1>, padding=<ltrb 5 5 0 4>}
   slot {class=GridSlot, properties {column=3, row=1}}
  }
  UEFN_TextBlock_C_6 {
   widgetClassPath=&UEFN_TextBlock
   parent=HeaderBorder_03
   widget=UEFN_TextBlock_C_6
   properties {
    text=Loudness Results
    font {
     bForceMonospaced=false
     bMaterialIsStencil=false
     fontMaterial=null
     fontObject=&BrutalType
     letterSpacing=100
     monospacedWidth=1
     size=21
     skewAmount=0
     typefaceFontName=Light
     outlineSettings {
      bApplyOutlineToDropShadows=false
      bMiteredCorners=false
      bSeparateFillAlpha=false
      outlineMaterial=null
      outlineSize=0
      outlineColor=<rgba 0 0 0 1>
     }
    }
   }
   slot {class=BorderSlot, properties {horizontalAlignment=HAlign_Center, verticalAlignment=VAlign_Center, padding=<ltrb 5 5 0 4>}}
  }
  Border_1 {
   widgetClassPath=&Border
   parent=MainGridPanel
   widget=Border_1
   properties {brushColor=<rgba 0.0208 0.0208 0.0208 1>}
   slot {class=GridSlot, properties {column=3, row=3}}
  }
  LoudnessResultsHeaderRowGridPanel {
   widgetClassPath=&GridPanel
   parent=MainGridPanel
   widget=LoudnessResultsHeaderRowGridPanel
   slot {class=GridSlot, properties {column=3, row=3}}
  }
  Spacer_SoundWaveName {
   widgetClassPath=&Spacer
   parent=LoudnessResultsHeaderRowGridPanel
   widget=Spacer_SoundWaveName
   properties {size=<502.2117 1>}
   slot {class=GridSlot, properties {layer=5}}
  }
  UEFN_TextBlock_C_5 {
   widgetClassPath=&UEFN_TextBlock
   parent=LoudnessResultsHeaderRowGridPanel
   widget=UEFN_TextBlock_C_5
   properties {
    text=Sound Wave
    font {
     bForceMonospaced=false
     bMaterialIsStencil=false
     fontMaterial=null
     fontObject=&BurbankSmall
     letterSpacing=0
     monospacedWidth=1
     size=12
     skewAmount=0
     typefaceFontName=Bold
     outlineSettings {
      bApplyOutlineToDropShadows=false
      bMiteredCorners=false
      bSeparateFillAlpha=false
      outlineMaterial=null
      outlineSize=0
      outlineColor=<rgba 0 0 0 1>
     }
    }
   }
   slot {class=GridSlot, properties {layer=5, verticalAlignment=VAlign_Bottom, padding=<ltrb 5 2 0 2>}}
  }
  Spacer_SoundWaveLoudness {
   widgetClassPath=&Spacer
   parent=LoudnessResultsHeaderRowGridPanel
   widget=Spacer_SoundWaveLoudness
   properties {size=<240 1>}
   slot {class=GridSlot, properties {column=1, columnSpan=3, layer=5}}
  }
  UEFN_TextBlock_C_168 {
   widgetClassPath=&UEFN_TextBlock
   parent=LoudnessResultsHeaderRowGridPanel
   widget=UEFN_TextBlock_C_168
   properties {
    text=Avg
    font {
     bForceMonospaced=false
     bMaterialIsStencil=false
     fontMaterial=null
     fontObject=&BurbankSmall
     letterSpacing=0
     monospacedWidth=1
     size=12
     skewAmount=0
     typefaceFontName=Bold
     outlineSettings {
      bApplyOutlineToDropShadows=false
      bMiteredCorners=false
      bSeparateFillAlpha=false
      outlineMaterial=null
      outlineSize=0
      outlineColor=<rgba 0 0 0 1>
     }
    }
   }
   slot {class=GridSlot, properties {column=1, layer=5, verticalAlignment=VAlign_Bottom, padding=<ltrb 0 2 0 2>}}
  }
  UEFN_TextBlock_C_3 {
   widgetClassPath=&UEFN_TextBlock
   parent=LoudnessResultsHeaderRowGridPanel
   widget=UEFN_TextBlock_C_3
   properties {
    text=Min
    font {
     bForceMonospaced=false
     bMaterialIsStencil=false
     fontMaterial=null
     fontObject=&BurbankSmall
     letterSpacing=0
     monospacedWidth=1
     size=12
     skewAmount=0
     typefaceFontName=Bold
     outlineSettings {
      bApplyOutlineToDropShadows=false
      bMiteredCorners=false
      bSeparateFillAlpha=false
      outlineMaterial=null
      outlineSize=0
      outlineColor=<rgba 0 0 0 1>
     }
    }
   }
   slot {class=GridSlot, properties {column=2, layer=5, verticalAlignment=VAlign_Bottom, padding=<ltrb 0 2 0 2>}}
  }
  UEFN_TextBlock_C_4 {
   widgetClassPath=&UEFN_TextBlock
   parent=LoudnessResultsHeaderRowGridPanel
   widget=UEFN_TextBlock_C_4
   properties {
    text=Max
    font {
     bForceMonospaced=false
     bMaterialIsStencil=false
     fontMaterial=null
     fontObject=&BurbankSmall
     letterSpacing=0
     monospacedWidth=1
     size=12
     skewAmount=0
     typefaceFontName=Bold
     outlineSettings {
      bApplyOutlineToDropShadows=false
      bMiteredCorners=false
      bSeparateFillAlpha=false
      outlineMaterial=null
      outlineSize=0
      outlineColor=<rgba 0 0 0 1>
     }
    }
   }
   slot {class=GridSlot, properties {column=3, layer=5, verticalAlignment=VAlign_Bottom, padding=<ltrb 0 2 0 2>}}
  }
  SoundWaveInfoSizeBox {
   widgetClassPath=&SizeBox
   parent=MainGridPanel
   widget=SoundWaveInfoSizeBox
   properties {
    bOverride_HeightOverride=true
    bOverride_MaxDesiredWidth=true
    bOverride_MinDesiredHeight=true
    bOverride_MinDesiredWidth=true
    maxDesiredWidth=700
    minDesiredWidth=700
   }
   slot {class=GridSlot, properties {column=3, layer=5, row=4, rowSpan=11}}
  }
  SoundWaveInfoBorder {
   widgetClassPath=&Border
   parent=SoundWaveInfoSizeBox
   widget=SoundWaveInfoBorder
   properties {brushColor=<rgba 0.0103 0.0103 0.0103 1>, padding=<ltrb 0 0 0 0>}
   slot {class=SizeBoxSlot}
  }
  SoundWaveScrollBox {
   widgetClassPath=&EditorUtilityScrollBox
   bIsVariable=true
   parent=SoundWaveInfoBorder
   widget=SoundWaveScrollBox
   properties {alwaysShowScrollbar=true, alwaysShowScrollbarTrack=true, scrollbarThickness=<15.181 9>}
   slot {class=BorderSlot, properties {padding=<ltrb 0 0 0 0>}}
  }
 ]
}
