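; MVVM view model, by its name, for a non-real-time loudness analyzer asset among the audio analyzer utilities.
; Its one function, GetSoundWaveNameFromNRT, reads the Sound from the stored LoudnessNRT variable and returns that sound's display name as text.
; Validity checks return localized fallback text, "Invalid NRT" or "Invalid Soundwave", when either reference is missing.
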(fn GetSoundWaveNameFromNRT
  (bind _loudnessnrt (Variables|LoudnessNRT|GetLoudnessNRT)
    (:then
      (bind _sound (Variables|/Script/AudioAnalyzer.AudioAnalyzerNRT|GetSound _loudnessnrt)
        (:then
          (return (Utilities|Text|ToText(String) (Utilities|GetDisplayName _sound))))
        (:else
          (return "NSLOCTEXT(\"[8633B60134DB1842BF81677320BD09A7]\", \"17EC1D9D48A9448C06E10CA180623280\", \"Invalid Soundwave\")"))))
    (:else
      (return "NSLOCTEXT(\"[8633B60134DB1842BF81677320BD09A7]\", \"22A5D395420093D106EC5BA08DD0B204\", \"Invalid NRT\")"))))
