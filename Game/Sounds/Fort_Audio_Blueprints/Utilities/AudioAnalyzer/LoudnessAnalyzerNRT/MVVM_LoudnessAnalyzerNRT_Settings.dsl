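; MVVM view model for LoudnessNRT analyzer settings in an audio analyzer utility, exposing the curve type as displayable text.
; Its one function, GetCurveTypeAsText, does a validated get of the LoudnessNRTSettings variable and converts that object's CurveType to text.
; The invalid-settings branch is empty, so no text value is set in that case.
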(fn GetCurveTypeAsText
  (bind _loudnessnrtsettings (Variables|LoudnessNRT|GetLoudnessNRTSettings)
    (:then
      (return (Utilities|Text|ToText (Variables|/Script/AudioSynesthesia.LoudnessNRTSettings|GetCurveType _loudnessnrtsettings))))
    (:else)))
