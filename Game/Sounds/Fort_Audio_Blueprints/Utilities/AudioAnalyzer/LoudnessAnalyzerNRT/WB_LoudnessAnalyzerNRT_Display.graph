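# Widget row for an audio analyzer utility that shows a LoudnessNRT asset's sound wave name and loudness values over a gradient fill.
# SetLoudnessNRT stores the asset, fires OnLoudnessNRTSet and refreshes the name, which falls back to default or "Invalid Soundwave" text.
# PreConstruct also refreshes the name.
# SetLoudnessData formats average, min and max to two decimals, but writes Min into Value_Max and Max into Value_Min.
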WidgetBlueprint /Game/Sounds/Fort_Audio_Blueprints/Utilities/AudioAnalyzer/LoudnessAnalyzerNRT/WB_LoudnessAnalyzerNRT_Display
graphs [
 EventGraph {nodes [
   LoudnessNRTSet() {is=title K2Node_CustomEvent}
   > HandleLoudnessNRTChanged()
   Event PreConstruct()
   > UpdateSoundWaveName()
   Bind Event to On Loudness NRTSet(self:WB Loudness Analyzer NRT Display Object Reference, Delegate=Create Event()) {is=title K2Node_AddDelegate}
   Event Construct()
  ]}
 GetLoudnessNRTSoundWaveName {nodes [
   Inputs!()
   n1 = Get LoudnessNRT() -> (LoudnessNRT:Loudness NRT Object Reference)
   n2 = Get Sound!(self:Audio Analyzer NRT Object Reference=@n1) -> (Sound:Sound Wave Object Reference)
   n3 = IsValid(Object=@n2)
   n4 = To Text!(Value:String=GetDisplayName(Object=@n2)) -> (ReturnValue:Text) {is=title K2Node_GenericToText}
   Outputs!(ReturnValue:Text=Select(
     Option 0:Text=Get DefaultSoundWaveName() -> (DefaultSoundWaveName:Text)
     Option 1:Text=Select(Option 0="LOCTEXT(\"Invalid Soundwave\")", Option 1=@n4, Index=@n3)
     Index=IsValid(Object=@n1)
    ) -> (ReturnValue:Text))
  ]}
 HandleLoudnessNRTChanged {nodes [
   HandleLoudnessNRTChanged() {is=title K2Node_FunctionEntry}
   > UpdateSoundWaveName()
  ]}
 OnLoudnessNRTSet {nodes [OnLoudnessNRTSet() {is=title K2Node_FunctionEntry}]}
 SetLoudnessData {nodes [
   n1 = SetLoudnessData!() -> (then, Average:double=~, Min:double=~, Max:double=~) {is=title K2Node_FunctionEntry}
   > Set Text(
    Text="To Text (Float)"(Value=@n1, MinimumIntegralDigits=1, MaximumIntegralDigits=324, MinimumFractionalDigits=2, MaximumFractionalDigits=2)
    self=Get Value_Avg() -> (Value_Avg:Text Block Object Reference)
   )
   > Set Text(
    Text="To Text (Float)"(Value=@n1.Min, MinimumIntegralDigits=1, MaximumIntegralDigits=324, MinimumFractionalDigits=2, MaximumFractionalDigits=2)
    self=Get Value_Max() -> (Value_Max:Text Block Object Reference)
   )
   > Set Text(
    Text="To Text (Float)"(Value=@n1.Max, MinimumIntegralDigits=1, MaximumIntegralDigits=324, MinimumFractionalDigits=2, MaximumFractionalDigits=2)
    self=Get Value_Min() -> (Value_Min:Text Block Object Reference)
   )
  ]}
 SetLoudnessNRT {nodes [
   n1 = SetLoudnessNRT!() -> (then, LoudnessNRT:Loudness NRT Object Reference) {is=title K2Node_FunctionEntry}
   > Set LoudnessNRT(LoudnessNRT:Loudness NRT Object Reference=@n1) -> (Output_Get:Loudness NRT Object Reference)
   > Call On Loudness NRTSet!(execute:Exec, self:WB Loudness Analyzer NRT Display Object Reference) -> (then:Exec) {is=title K2Node_CallDelegate}
   > HandleLoudnessNRTChanged() {is=title K2Node_CallFunction}
  ]}
 UpdateSoundWaveName {nodes [
   UpdateSoundWaveName() {is=title K2Node_FunctionEntry}
   > n1 = Set SoundWaveName(SoundWaveName:Text=Get Loudness NRTSound Wave Name!() -> (ReturnValue:Text) {is=title K2Node_MacroInstance}) -> (Output_Get:Text)
   > Set Text(Text=@n1, self=Get Text_SoundWaveName() -> (Text_SoundWaveName:Text Block Object Reference))
  ]}
]
variables [
 DefaultSoundWaveName {type=string, default="< Sound Wave >"}
 LoudnessNRT {type=LoudnessNRT, default=null}
 OnLoudnessNRTSet {}
 SoundWaveName {type=string, default=""}
]
widgets {
 info {inheritedWidgetCount=0, parentClass=&UserWidget, rootWidgetClass=&GridPanel, widgetCount=9}
 widgets [
  GridPanel_32 {widgetClassPath=&GridPanel, widget=GridPanel_32}
  BasicGradientFill {
   widgetClassPath=/Game/UI/Foundation/Stats/BasicGradientFill
   bIsVariable=true
   parent=GridPanel_32
   widget=BasicGradientFill
   properties {rotation=0.456, color 1=<rgba 0 0.1697 0.75 1>, color 2=<rgba 0 0.6462 0.7708 1>}
   slot {class=GridSlot, properties {columnSpan=10}}
  }
  Text_SoundWaveName {
   widgetClassPath=&UEFN_TextBlock
   bIsVariable=true
   parent=GridPanel_32
   widget=Text_SoundWaveName
   properties {
    text="< SoundWave Name >"
    font {
     bForceMonospaced=false
     bMaterialIsStencil=false
     fontMaterial=null
     fontObject=&BurbankSmall
     letterSpacing=0
     monospacedWidth=1
     size=9
     skewAmount=0
     typefaceFontName=Black
     outlineSettings {
      bApplyOutlineToDropShadows=false
      bMiteredCorners=false
      bSeparateFillAlpha=false
      outlineMaterial=null
      outlineSize=0
      outlineColor=<rgba 0 0 0 1>
     }
    }
   }
   slot {class=GridSlot, properties {horizontalAlignment=HAlign_Left, padding=<ltrb 4 2 0 2>}}
  }
  Value_Avg {
   widgetClassPath=&UEFN_TextBlock
   bIsVariable=true
   parent=GridPanel_32
   widget=Value_Avg
   properties {
    text=0.00
    font {
     bForceMonospaced=false
     bMaterialIsStencil=false
     fontMaterial=null
     fontObject=&BurbankSmall
     letterSpacing=0
     monospacedWidth=1
     size=9
     skewAmount=0
     typefaceFontName=Black
     outlineSettings {
      bApplyOutlineToDropShadows=false
      bMiteredCorners=false
      bSeparateFillAlpha=false
      outlineMaterial=null
      outlineSize=0
      outlineColor=<rgba 0 0 0 1>
     }
    }
   }
   slot {class=GridSlot, properties {column=3, horizontalAlignment=HAlign_Center, padding=<ltrb 0 2 0 0>}}
  }
  Value_Min {
   widgetClassPath=&UEFN_TextBlock
   bIsVariable=true
   parent=GridPanel_32
   widget=Value_Min
   properties {
    text=0.00
    font {
     bForceMonospaced=false
     bMaterialIsStencil=false
     fontMaterial=null
     fontObject=&BurbankSmall
     letterSpacing=0
     monospacedWidth=1
     size=9
     skewAmount=0
     typefaceFontName=Black
     outlineSettings {
      bApplyOutlineToDropShadows=false
      bMiteredCorners=false
      bSeparateFillAlpha=false
      outlineMaterial=null
      outlineSize=0
      outlineColor=<rgba 0 0 0 1>
     }
    }
   }
   slot {class=GridSlot, properties {column=5, horizontalAlignment=HAlign_Center, padding=<ltrb 0 2 0 0>}}
  }
  Value_Max {
   widgetClassPath=&UEFN_TextBlock
   bIsVariable=true
   parent=GridPanel_32
   widget=Value_Max
   properties {
    text=0.00
    font {
     bForceMonospaced=false
     bMaterialIsStencil=false
     fontMaterial=null
     fontObject=&BurbankSmall
     letterSpacing=0
     monospacedWidth=1
     size=9
     skewAmount=0
     typefaceFontName=Black
     outlineSettings {
      bApplyOutlineToDropShadows=false
      bMiteredCorners=false
      bSeparateFillAlpha=false
      outlineMaterial=null
      outlineSize=0
      outlineColor=<rgba 0 0 0 1>
     }
    }
   }
   slot {class=GridSlot, properties {column=7, horizontalAlignment=HAlign_Center, padding=<ltrb 0 2 0 0>}}
  }
  Spacer_70 {
   widgetClassPath=&Spacer
   parent=GridPanel_32
   widget=Spacer_70
   properties {size=<500 10.6387>}
   slot {class=GridSlot, properties {padding=<ltrb 0 2 0 0>}}
  }
  Spacer_2 {
   widgetClassPath=&Spacer
   parent=GridPanel_32
   widget=Spacer_2
   properties {size=<200 10.6387>}
   slot {class=GridSlot, properties {column=3, columnSpan=5}}
  }
  Spacer {
   widgetClassPath=&Spacer
   parent=GridPanel_32
   widget=Spacer
   properties {size=<50 10.6387>}
   slot {class=GridSlot, properties {column=8}}
  }
 ]
}
